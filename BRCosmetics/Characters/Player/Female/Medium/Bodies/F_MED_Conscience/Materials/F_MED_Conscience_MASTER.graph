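# Master character material for the F_MED_Conscience female body cosmetic, with animated ribbons and reactive styling.
# It modifies base color, roughness, normal and emissive between shared PreFX and PostFX functions.
# Two panning noises distort ribbon masks that also generate normals, while progress parameters build reactive masks, alongside pulse, flash and fresnel glow.
# The IsEyeballMaterial static switch selects dither-masked eyes with rippling eyeliner and time-hashed pupil jumps.
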Material /BRCosmetics/Characters/Player/Female/Medium/Bodies/F_MED_Conscience/Materials/F_MED_Conscience_MASTER
graph {
 properties {
  bAutomaticallySetUsageInEditor=false
  bUseMaterialAttributes=true
  bUsedWithMeshDeformer=true
  bUsedWithMorphTargets=true
  bUsedWithSkeletalMesh=true
  pixelDepthOffsetMode=PDOM_Legacy
  refractionMethod=RM_IndexOfRefraction
  shadingModel=MSM_FromMaterialExpression
  subsurfaceProfile=&SS_HeroSkin_02
  usedShadingModels=Default Lit | Subsurface | Clear Coat | Subsurface Profile | Cloth | Eye
 }
 nodes [
  Constant() {r=1}
  n1 = MaterialFunctionCall() {materialFunction=&BaseFN_CharacterFunction_PreFX}
  n2 = GetMaterialAttributes(None=MaterialFunctionCall(
    PreskinnedPosition=@n1.Pre skinned Local Position
    Gmap=@n1.Pre skinned Local Position
    MaterialAttributes=@n1.Pre skinned Local Position
    AO=@n1.Pre skinned Local Position
   ) {materialFunction=&CharacterSkinFX}.Result) {attributeGetTypes [
    #1
    #2
    #3
    #4
   ]}
  n3 = ScalarParameter() {defaultValue=0.49, group=Conscience, parameterName=Rough1}
  n4 = ScalarParameter() {defaultValue=0.18, group=Conscience, parameterName=Rough2}
  n5 = NamedRerouteUsage()
  n6 = SetMaterialAttributes!(
   MaterialAttributes=@n2.MaterialAttributes
   Base Color=LinearInterpolate(A=NamedRerouteUsage(), B=NamedRerouteUsage(), Alpha=NamedRerouteUsage())
   Roughness=LinearInterpolate(A=@n2.Roughness, B=LinearInterpolate(A=@n3, B=@n4, Alpha=@n5), Alpha=NamedRerouteUsage())
   Normal=LinearInterpolate(A=@n2.Normal, B=NamedRerouteUsage(), Alpha=NamedRerouteUsage())
   Emissive Color=Add(A=@n2.Emissive Color, B=NamedRerouteUsage())
  ) {attributeSetTypes [
    #1
    #2
    #3
    #4
   ]}
  n7 = MaterialFunctionCall(
   AO=@n1.AO
   Distance Blend=@n1.AO
   Fresnel=@n1.AO
   Gmap=@n1.AO
   Pre Skinned Local Position=@n1.AO
   Material=StaticSwitchParameter(True=NamedRerouteUsage(), False=@n6) {group=Conscience, parameterName=IsEyeballMaterial}
  ) {materialFunction=&BaseFN_CharacterFunction_PostFX}
  MaterialFunctionCall(In=@n7.Result) {materialFunction=&CINE_MF_CharPostFX}
  n8 = VertexColor()
  Multiply(A=@n8.R, B=ScalarParameter() {defaultValue=0.4, group=Conscience, parameterName=DEBUG_Rotation})
  NamedRerouteDeclaration(None=@n1.Pre skinned Local Position) {name=PreskinnedPos, nodeColor=<rgba 0.4062 1 0 1>}
  n9 = NamedRerouteUsage()
  n10 = Time()
  n11 = ScalarParameter() {defaultValue=0.85, group=Conscience, parameterName=Reactive_Pulse_Speed}
  n12 = ConstantBiasScale(None=Sine(None=Multiply(A=@n10, B=@n11) {constB=0.75}))
  NamedRerouteDeclaration(None=Saturate(None=Multiply(A=@n9, B=@n12))) {name=Reactive_PulseMask, nodeColor=<rgba 1 0 0.8203 1>}
  n13 = TextureCoordinate()
  n14 = ScalarParameter() {defaultValue=10, group=Conscience, parameterName=DistortionB_Tiling}
  n15 = ScalarParameter() {defaultValue=0.15, group=Conscience, parameterName=DistorationB_Speed}
  n16 = TextureSample(UVs=Panner(
    Coordinate=Multiply(A=@n13, B=@n14)
    Speed=MaterialFunctionCall(Y=@n15) {materialFunction=&MakeFloat2}.Result
   ) {speedY=0.02}) {samplerSource=SSM_Wrap_WorldGroupSettings, texture=/Game/Global/Textures/Default/Noise/T_Noise_Cloud}
  NamedRerouteDeclaration(None=@n16.R) {name=DistortionB_Noise, nodeColor=<rgba 0.4766 1 0 1>}
  n17 = TextureSample() {
   samplerSource=SSM_Wrap_WorldGroupSettings
   samplerType=SAMPLERTYPE_Masks
   texture=/BRCosmetics/Characters/Player/Female/Medium/Bodies/F_MED_Conscience/Textures/T_F_MED_Conscience_Body_Skin_FX
  }
  n18 = NamedRerouteDeclaration(None=OneMinus(None=@n17.A)) {name=FX_Mask, nodeColor=<rgba 0.9922 0 1 1>}
  n19 = Multiply(A=ComponentMask(None=@n16.RGB) {g=true, r=true}, B=ScalarParameter() {defaultValue=0.01, group=Conscience, parameterName=DistortionB_Amount})
  n20 = TextureCoordinate()
  n21 = ScalarParameter() {defaultValue=2.5, group=Conscience, parameterName=DistortionA_Tiling}
  n22 = ScalarParameter() {defaultValue=0.05, group=Conscience, parameterName=DistorationA_Speed}
  n23 = TextureSample(UVs=Panner(
    Coordinate=Multiply(A=@n20, B=@n21)
    Speed=MaterialFunctionCall(Y=@n22) {materialFunction=&MakeFloat2}.Result
   ) {speedY=0.02}) {
   samplerSource=SSM_Wrap_WorldGroupSettings
   samplerType=SAMPLERTYPE_LinearColor
   texture=/Game/Creative/Sets/PropSets/Primitives/Customization/Textures/GalacticSwirl/T_CM_GalacticSwirl_Flowmap
  }
  n24 = Multiply(
   A=ConstantBiasScale(None=ComponentMask(None=@n23.RGB) {g=true, r=true}) {bias=-0.5, scale=2}
   B=ScalarParameter() {defaultValue=0.03, group=Conscience, parameterName=DistortionA_Amount}
  )
  n25 = TextureCoordinate()
  n26 = Add(A=Multiply(A=@n18, B=Add(A=@n24, B=@n19)), B=@n25)
  n27 = TextureSample(UVs=@n26) {
   samplerSource=SSM_Wrap_WorldGroupSettings
   samplerType=SAMPLERTYPE_Masks
   texture=&T_F_MED_Conscience_Body_Skin_FX
  }
  n28 = Multiply(A=@n24, B=Max(A=@n16.G) {constB=0.5})
  n29 = TextureSample(UVs=Add(A=@n25, B=Multiply(A=@n18, B=@n28))) {
   samplerSource=SSM_Wrap_WorldGroupSettings
   samplerType=SAMPLERTYPE_Masks
   texture=&T_F_MED_Conscience_Body_Skin_FX
  }
  n30 = Add(A=@n27.G, B=@n29.R)
  n31 = NamedRerouteUsage()
  NamedRerouteDeclaration(None=Saturate(None=Multiply(A=@n30, B=@n31))) {name=Ribbon_Mask, nodeColor=<rgba 0.4297 0 1 1>}
  n32 = TextureObject() {
   samplerType=SAMPLERTYPE_Masks
   texture=&T_F_MED_Conscience_Body_Skin_FX
  }
  n33 = VertexNormalWS()
  n34 = StaticBool() {value=true}
  n35 = Constant() {r=3}
  n36 = Multiply(
   A=ScalarParameter() {defaultValue=0.5, group=Conscience, parameterName=GeneratedNormal_Intensity}
   B=Saturate(None=MaterialFunctionCall(Normal Vector=@n33, Invert Fresnel=@n34, Power=@n35) {materialFunction=&Fresnel_Function}.Result)
  )
  n37 = NamedRerouteUsage()
  NamedRerouteDeclaration(None=MaterialFunctionCall!(
    Height Map=@n32
    Normal Map Intensity=Multiply(A=@n36, B=@n37)
    Height Map UV Offset=ScalarParameter() {defaultValue=0.003, group=Conscience, parameterName=GeneratedNormal_Distance}
    Coordinates=@n26
    Height Map Channel Selector=Constant4Vector() {constant=<rgba 0 1 0 0>}.RGBA
   ) {materialFunction=/Engine/Functions/Engine_MaterialFunctions03/Procedurals/NormalFromHeightmap}.Result) {name=Generated_Ribbon_Normals, nodeColor=<rgba 0.0781 0 1 1>}
  NamedRerouteDeclaration(None=@n27.G) {name=Ribbon_G_Mask, nodeColor=<rgba 1 0.75 0 1>}
  n38 = NamedRerouteUsage()
  n39 = LinearInterpolate(
   A=VectorParameter() {group=Conscience, parameterName=Ribbon_Color1, defaultValue=<rgba 0.2195 0.1746 0.1022 1>}.RGB
   B=VectorParameter() {group=Conscience, parameterName=Ribbon_Color2, defaultValue=<rgba 0.5451 0.0667 0.1255 1>}.RGB
   Alpha=@n38
  )
  n40 = LinearInterpolate(
   A=VectorParameter() {group=Conscience, parameterName=Ribbon_Red_Color1, defaultValue=<rgba 0.2031 0.2031 0.2031 1>}.RGB
   B=VectorParameter() {group=Conscience, parameterName=Ribbon_Red_Color2, defaultValue=<rgba 0.5451 0.0667 0.1255 1>}.RGB
   Alpha=@n38
  )
  n41 = Saturate(None=NamedRerouteUsage())
  n42 = Multiply(A=NamedRerouteUsage(), B=ScalarParameter() {defaultValue=0.05, group=Conscience, parameterName=ColorBreakupStrength}) {constB=0.05}
  n43 = NamedRerouteUsage()
  NamedRerouteDeclaration(None=Subtract(A=LinearInterpolate(A=@n39, B=@n40, Alpha=@n41), B=Multiply(A=@n42, B=@n43))) {name=Ribbon_BaseColor, nodeColor=<rgba 0.5 1 0 1>}
  NamedRerouteDeclaration(None=@n2.Base Color) {name=BaseColor_IN, nodeColor=<rgba 0.7578 1 0 1>}
  n44 = TextureSample() {
   samplerSource=SSM_Wrap_WorldGroupSettings
   texture=/BRCosmetics/Characters/Player/Female/Medium/Bodies/F_MED_Conscience/Textures/T_F_MED_Conscience_Body_Red_D
  }
  n45 = NamedRerouteUsage()
  n46 = NamedRerouteUsage()
  n47 = Ceil(None=Add(A=Subtract(A=@n45), B=Multiply(A=@n46) {constB=2}))
  NamedRerouteDeclaration(None=LinearInterpolate(A=NamedRerouteUsage(), B=@n44.RGB, Alpha=Saturate(None=@n47))) {name=BaseColor_OUT, nodeColor=<rgba 0.1953 1 0 1>}
  n48 = NamedRerouteUsage()
  n49 = ScalarParameter() {defaultValue=3, group=Conscience, parameterName=Reactive_Flash_Intensity}
  n50 = ComponentMask(None=NamedRerouteUsage()) {g=true}
  n51 = Multiply(A=Multiply(A=NamedRerouteUsage(), B=NamedRerouteUsage()), B=Multiply(A=Multiply(A=@n48, B=@n49), B=Abs(None=@n50)))
  n52 = TextureSampleParameter2D() {
   parameterName=SkinFX_Mask
   samplerSource=SSM_Wrap_WorldGroupSettings
   samplerType=SAMPLERTYPE_Masks
   texture=&T_F_MED_Conscience_Body_Skin_FX
  }
  n53 = Multiply(A=NamedRerouteUsage(), B=@n52.R)
  NamedRerouteDeclaration(None=Multiply(A=Add(A=@n51, B=@n53), B=NamedRerouteUsage())) {name=Emissive_OUT, nodeColor=<rgba 1 0.2578 0 1>}
  n54 = VectorParameter() {group=Conscience, parameterName=EdgeGlowColor, defaultValue=<rgba 0.2051 0 0.3021 1>}
  n55 = ScalarParameter() {defaultValue=4, group=Conscience, parameterName=EdgeGlow_Exp}
  n56 = NamedRerouteUsage()
  n57 = NamedRerouteDeclaration(None=Fresnel(ExponentIn=LinearInterpolate(A=@n55, Alpha=@n56) {constB=2.5}, Normal=VertexNormalWS()) {baseReflectFraction=0, exponent=3}) {name=EdgeGlow_Mask, nodeColor=<rgba 0.2188 0 1 1>}
  n58 = ScalarParameter() {defaultValue=3, group=Conscience, parameterName=EdgeGlow_Intensity}
  n59 = ComponentMask(None=NamedRerouteUsage()) {b=true}
  n60 = Saturate(None=Divide(A=Subtract(A=@n59) {constB=10}) {constB=35})
  NamedRerouteDeclaration(None=MaterialFunctionCall(LDR Input=Multiply(A=@n54.RGB, B=@n57), HDR Intensity Multiplier=Multiply(A=@n58, B=@n60)) {materialFunction=&DeriveHDRfromLDR}.RemappedRange) {name=EdgeGlow, nodeColor=<rgba 0 0.7656 1 1>}
  n61 = NamedRerouteUsage()
  n62 = LinearInterpolate(
   A=VectorParameter() {group=Conscience, parameterName=Eyeliner_Color, defaultValue=<rgba 0.0417 0.0334 0.0338 0>}.RGB
   B=VectorParameter() {group=Conscience, parameterName=Sclera_Color, defaultValue=<rgba 0.4271 0.2469 0.3615 1>}.RGB
   Alpha=@n61
  )
  n63 = Divide(A=NamedRerouteUsage(), B=ScalarParameter() {defaultValue=130, group=Conscience, parameterName=EyeOffseWS}) {constB=125}
  n64 = MaterialFunctionCall(
   Seed=Multiply(A=Time(), B=ScalarParameter() {defaultValue=0.2, group=Conscience, parameterName=OffsetSpeed})
   Vec2=Ceil(None=ComponentMask(None=@n63) {b=true, r=true})
  ) {materialFunction=&MF_Hash}
  n65 = Add(
   A=MaterialFunctionCall(UVs=NamedRerouteUsage(), Texture Scale=ScalarParameter() {defaultValue=0.4, group=Conscience, parameterName=EyeScale}) {materialFunction=&ScaleUVsByCenter}.UVs
   B=Multiply(A=Round(None=@n64.Vec2), B=ScalarParameter() {defaultValue=0.25, group=Conscience, parameterName=OffsetClamp})
  )
  n66 = TextureSample(UVs=@n65) {
   samplerSource=SSM_Clamp_WorldGroupSettings
   texture=/BRCosmetics/Characters/Player/Female/Medium/Bodies/F_MED_Conscience/Textures/T_F_MED_Conscience_Eyes_Plane_D
  }
  n67 = Multiply(A=@n66.R, B=VectorParameter() {group=Conscience, parameterName=EyeColor, defaultValue=<rgba 0.2122 0.1022 0.1356 1>}.RGB)
  n68 = Multiply(A=@n61, B=SphereMask(A=Constant2Vector() {g=0.5, r=0.5}, B=@n65) {attenuationRadius=0.42, hardnessPercent=90})
  n69 = TextureCoordinate()
  n70 = ScalarParameter() {defaultValue=1.1, group=Conscience, parameterName=Eyeliner_Distortion_Tiling}
  n71 = ScalarParameter() {defaultValue=0.3, group=Conscience, parameterName=Eyeliner_Distortion_Speed}
  n72 = TextureSample(UVs=Panner(
    Coordinate=Multiply(A=@n69, B=@n70)
    Speed=MaterialFunctionCall(X=@n71) {materialFunction=&MakeFloat2}.Result
   ) {speedY=0.02}) {
   samplerSource=SSM_Wrap_WorldGroupSettings
   samplerType=SAMPLERTYPE_LinearColor
   texture=&T_CM_GalacticSwirl_Flowmap
  }
  n73 = MaterialFunctionCall(Float2=ConstantBiasScale(None=@n72.RGB) {bias=-0.5, scale=2}) {materialFunction=&BreakOutFloat2Components}
  n74 = Multiply(A=@n73.R, B=ScalarParameter() {defaultValue=0.13, group=Conscience, parameterName=Eyeliner_Distortion_X})
  n75 = Multiply(A=@n73.G, B=ScalarParameter() {defaultValue=0.07, group=Conscience, parameterName=Eyeliner_Distortion_Y})
  n76 = NamedRerouteUsage()
  n77 = TextureSample(UVs=Add(A=MaterialFunctionCall(X=@n74, Y=@n75) {materialFunction=&MakeFloat2}.Result, B=@n76)) {
   samplerSource=SSM_Clamp_WorldGroupSettings
   samplerType=SAMPLERTYPE_Masks
   texture=/BRCosmetics/Characters/Player/Female/Medium/Bodies/F_MED_Conscience/Textures/T_F_MED_Conscience_Eyes_Plane_FX
  }
  n78 = TextureSample(UVs=@n76) {
   samplerSource=SSM_Clamp_WorldGroupSettings
   samplerType=SAMPLERTYPE_Masks
   texture=&T_F_MED_Conscience_Eyes_Plane_FX
  }
  n79 = Add(A=Add(A=@n77.R, B=@n78.R), B=NamedRerouteDeclaration(None=@n78.B) {name=Sclera_Mask, nodeColor=<rgba 1 0.2812 0 1>})
  NamedRerouteDeclaration(None=MakeMaterialAttributes!(
    BaseColor=LinearInterpolate(A=@n62, B=@n67, Alpha=@n68)
    Metallic
    Specular=Constant()
    Roughness=ScalarParameter() {defaultValue=0.33, group=Conscience, parameterName=Eye_Roughness}
    Anisotropy
    EmissiveColor=NamedRerouteUsage()
    Opacity
    OpacityMask=MaterialFunctionCall(Alpha Threshold=@n79) {materialFunction=&DitherTemporalAA_Fortnite}.Result
    Normal
    Tangent
    WorldPositionOffset
    SubsurfaceColor
    ClearCoat
    ClearCoatRoughness
    AmbientOcclusion
    Refraction
    CustomizedUVs_0
    CustomizedUVs_1
    CustomizedUVs_2
    CustomizedUVs_3
    CustomizedUVs_4
    CustomizedUVs_5
    CustomizedUVs_6
    CustomizedUVs_7
    PixelDepthOffset
    ShadingModel
    Displacement
   )) {name=Eye_Attribs, nodeColor=<rgba 0 0.1328 1 1>}
  NamedRerouteDeclaration(None=Rotator(Coordinate=TextureCoordinate(), Time=@n8.R) {speed=6.2832}) {name=Rotated_Eye_UVs, nodeColor=<rgba 0.9688 0 1 1>}
  n80 = MaterialFunctionCall(Float2=TextureCoordinate()) {materialFunction=&BreakOutFloat2Components}
  n81 = Ceil(None=Subtract(A=@n80.R, B=ScalarParameter() {defaultValue=0.572, group=Conscience, parameterName=BoxMask_X}))
  n82 = ScalarParameter() {defaultValue=0.315, group=Conscience, parameterName=BoxMask_Y}
  n83 = OneMinus(None=Ceil(None=Subtract(A=@n80.G, B=@n82)))
  NamedRerouteDeclaration(None=OneMinus(None=Multiply(A=@n81, B=@n83))) {name=Vest_Mask, nodeColor=<rgba 0.3359 1 0 1>}
  n84 = ScalarParameter() {defaultValue=1, group=Conscience, parameterName=IsReactive}
  NamedRerouteDeclaration(None=Multiply(A=ScalarParameter() {group=Conscience, parameterName=Reactive_Progress_A}, B=@n84)) {name=Reactive_Step1_Mask, nodeColor=<rgba 0 0.0391 1 1>}
  NamedRerouteDeclaration(None=Multiply(A=ScalarParameter() {group=Conscience, parameterName=Reactive_Progress_B}, B=@n84)) {name=Reactive_Step2_Mask, nodeColor=<rgba 0 1 0.5781 1>}
  NamedRerouteDeclaration(None=Multiply(A=ScalarParameter() {group=Conscience, parameterName=Reactive_Progress_C}, B=@n84)) {name=Reactive_Step3_Mask, nodeColor=<rgba 0 1 0.0391 1>}
  n85 = Subtract(A=NamedRerouteUsage())
  n86 = Multiply(A=NamedRerouteUsage()) {constB=2}
  Saturate(None=Ceil(None=Add(A=@n85, B=@n86)))
  VertexNormalWS()
 ]
}
