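# Physics asset for the dangling straps on the F_MED_Conscience female player body cosmetic.
# Four chains (front and back, left and right) run from a master bone through two segments, all built from small capsules.
# Constraints link each segment to its parent with tight twist and swing limits, leaving most freedom on one swing axis.
# Two larger spine capsules, unconstrained to the chains, supply torso volumes for the straps to collide against.
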PhysicsAsset /BRCosmetics/Characters/Player/Female/Medium/Bodies/F_MED_Conscience/Meshes/F_MED_Conscience_Straps_Physics
bodies {
 dyn_strap_back_01_l [
  {shapeType=Capsule, length=0.9009, radius=0.5, shapeName=dyn_strap_front_01_l_capsule, center=<0.46 -0.1644 0>}
  {shapeType=Capsule, length=0.9009, radius=0.5, shapeName=dyn_strap_front_01_l_capsule, center=<1.4491 -0.1644 0>}
 ]
 dyn_strap_back_01_r [
  {shapeType=Capsule, length=0.9009, radius=0.5, shapeName=dyn_strap_front_01_l_capsule, center=<0.46 -0.1644 0>}
  {shapeType=Capsule, length=0.9009, radius=0.5, shapeName=dyn_strap_front_01_l_capsule, center=<1.4491 -0.1644 0>}
 ]
 dyn_strap_back_02_l [
  {shapeType=Capsule, length=1.4609, radius=1.1431, shapeName=dyn_strap_front_mstr_l_capsule, center=<1.2662 -0.7036 0>}
  {shapeType=Capsule, length=1.0938, radius=0.6594, shapeName=dyn_strap_front_mstr_l_capsule, center=<2.9076 -0.6116 0>}
 ]
 dyn_strap_back_02_r [
  {shapeType=Capsule, length=1.4609, radius=1.1431, shapeName=dyn_strap_front_mstr_l_capsule, center=<1.2662 -0.7036 0>}
  {shapeType=Capsule, length=1.0938, radius=0.6594, shapeName=dyn_strap_front_mstr_l_capsule, center=<2.9076 -0.6116 0>}
 ]
 dyn_strap_back_mstr_l [{shapeType=Capsule, length=0.7972, radius=0.517, shapeName=dyn_strap_back_mstr_l_capsule}]
 dyn_strap_back_mstr_r [{shapeType=Capsule, length=0.7972, radius=0.517, shapeName=dyn_strap_back_mstr_l_capsule, rotation=<rot 0 0 -180>}]
 dyn_strap_front_01_l [
  {shapeType=Capsule, length=0.9009, radius=0.5, shapeName=dyn_strap_front_01_l_capsule, center=<0.46 -0.1644 0>}
  {shapeType=Capsule, length=0.9009, radius=0.5, shapeName=dyn_strap_front_01_l_capsule, center=<1.4491 -0.1644 0>}
 ]
 dyn_strap_front_01_r [
  {shapeType=Capsule, length=0.9009, radius=0.5, shapeName=dyn_strap_front_01_l_capsule, center=<0.46 -0.1644 0>}
  {shapeType=Capsule, length=0.2927, radius=0.5, shapeName=dyn_strap_front_01_l_capsule, center=<1.4491 -0.1644 0>}
 ]
 dyn_strap_front_02_l [
  {shapeType=Capsule, length=1.4609, radius=1.1431, shapeName=dyn_strap_front_mstr_l_capsule, center=<1.2662 -0.7036 0>}
  {shapeType=Capsule, length=1.0938, radius=0.6594, shapeName=dyn_strap_front_mstr_l_capsule, center=<2.9076 -0.6115 0>}
 ]
 dyn_strap_front_02_r [
  {shapeType=Capsule, length=1.4609, radius=1.1431, shapeName=dyn_strap_front_mstr_l_capsule, center=<1.2662 -0.7036 0>}
  {shapeType=Capsule, length=1.0938, radius=0.6594, shapeName=dyn_strap_front_mstr_l_capsule, center=<2.9076 -0.6116 0>}
 ]
 dyn_strap_front_mstr_l [{shapeType=Capsule, length=0.6909, radius=0.5, shapeName=dyn_strap_front_mstr_l_capsule}]
 dyn_strap_front_mstr_r [{shapeType=Capsule, length=0.6909, radius=0.5, shapeName=dyn_strap_front_mstr_l_capsule}]
 spine_02 [{shapeType=Capsule, length=5.902, radius=8.7576, shapeName=spine_02_capsule, center=<0 1.7884 -0.1967>, rotation=<rot 90 -0.0009 -0.0002>}]
 spine_03 [{shapeType=Capsule, length=11.6966, radius=5.7902, shapeName=spine_03_capsule, center=<-1.9775 2.1879 0>}]
}
constraints [
 {bone1Name=dyn_strap_front_01_l, bone2Name=dyn_strap_front_mstr_l, swing1LimitDegrees=15, swing2LimitDegrees=2.5, twistLimitDegrees=2.5}
 {bone1Name=dyn_strap_back_01_l, bone2Name=dyn_strap_back_mstr_l, swing1LimitDegrees=15, swing2LimitDegrees=2.5, twistLimitDegrees=2.5}
 {bone1Name=dyn_strap_back_01_r, bone2Name=dyn_strap_back_mstr_r, swing1LimitDegrees=12.5, swing2LimitDegrees=2.5, twistLimitDegrees=2.5}
 {bone1Name=dyn_strap_front_01_r, bone2Name=dyn_strap_front_mstr_r, swing1LimitDegrees=15, swing2LimitDegrees=2.5, twistLimitDegrees=2.5}
 {bone1Name=dyn_strap_front_02_l, bone2Name=dyn_strap_front_01_l, swing1LimitDegrees=10, swing2LimitDegrees=2.5, twistLimitDegrees=2.5}
 {bone1Name=dyn_strap_front_02_r, bone2Name=dyn_strap_front_01_r, swing1LimitDegrees=10, swing2LimitDegrees=2.5, twistLimitDegrees=2.5}
 {bone1Name=dyn_strap_back_02_l, bone2Name=dyn_strap_back_01_l, swing1LimitDegrees=15, swing2LimitDegrees=2.5, twistLimitDegrees=2.5}
 {bone1Name=dyn_strap_back_02_r, bone2Name=dyn_strap_back_01_r, swing1LimitDegrees=15, swing2LimitDegrees=2.5, twistLimitDegrees=2.5}
]
