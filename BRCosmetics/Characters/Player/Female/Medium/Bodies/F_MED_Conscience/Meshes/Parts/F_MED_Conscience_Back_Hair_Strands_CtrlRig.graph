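# Control Rig that drives the left and right back hair strands of the F_MED_Conscience character from head and upper-body motion.
# Top controls ride the head, while 07 end controls follow a spine_05/clavicle blend with maintained offset.
# Nulls 04–06 blend between controls 03 and 07 before 07 updates, then non-stretch B-spline IK poses each chain.
# The 07 constraint targets the control on the left but its null on the right.
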ControlRigBlueprint /BRCosmetics/Characters/Player/Female/Medium/Bodies/F_MED_Conscience/Meshes/Parts/F_MED_Conscience_Back_Hair_Strands_CtrlRig
graphs [
 RigVMFunctionLibrary {links [], nodes []}
 RigVMModel {nodes [
   RigUnit_BeginExecution = &RigUnit_BeginExecution()
   SequenceAggregate = &RigVMFunction_Sequence(ExecuteContext=@RigUnit_BeginExecution.ExecutePin)
   /Script/ControlRig.RigUnit_PositionConstraintLocalSpaceOffset(
    Child=(Type=Null,Name="hair_strand_back_04_l_ctrl_null")
    Filter=(bX=True,bY=True,bZ=True)
    Parents=((Item=(Type=Control,Name="hair_strand_back_03_l_ctrl"),Weight=0.75),(Item=(Type=Control,Name="hair_strand_back_07_l_ctrl"),Weight=0.25))
    Weight=1
    bMaintainOffset=false
    ExecutePin=@SequenceAggregate.A
   ) {is=struct RigVMUnitNode}
   > &RigUnit_PositionConstraintLocalSpaceOffset(
    Child=(Type=Null,Name="hair_strand_back_05_l_ctrl_null")
    Filter=(bX=True,bY=True,bZ=True)
    Parents=((Item=(Type=Control,Name="hair_strand_back_03_l_ctrl"),Weight=0.525),(Item=(Type=Control,Name="hair_strand_back_07_l_ctrl"),Weight=0.475))
    Weight=1
    bMaintainOffset=false
   )
   > &RigUnit_PositionConstraintLocalSpaceOffset(
    Child=(Type=Null,Name="hair_strand_back_06_l_ctrl_null")
    Filter=(bX=True,bY=True,bZ=True)
    Parents=((Item=(Type=Control,Name="hair_strand_back_03_l_ctrl"),Weight=0.25),(Item=(Type=Control,Name="hair_strand_back_07_l_ctrl"),Weight=0.75))
    Weight=1
    bMaintainOffset=false
   )
   > &RigUnit_PositionConstraintLocalSpaceOffset(
    Child=(Type=Control,Name="hair_strand_back_07_l_ctrl")
    Filter=(bX=True,bY=True,bZ=True)
    Parents=((Item=(Type=Bone,Name="spine_05"),Weight=0.8),(Item=(Type=Bone,Name="clavicle_l"),Weight=0.2))
    Weight=1
    bMaintainOffset=True
   )
   > SplineIK = RigVMFunctionReferenceNode(
    Bones="((Type=Bone,Name=\"hair_strand_back_01_l\"),(Type=Bone,Name=\"hair_strand_back_02_l\"),(Type=Bone,Name=\"hair_strand_back_03_l\"),(Type=Bone,Name=\"hair_strand_back_04_l\"),(Type=Bone,Name=\"hair_strand_back_05_l\"),(Type=Bone,Name=\"hair_strand_back_06_l\"),(Type=Bon ...(truncated)"
    Controls="((Type=Control,Name=\"hair_strand_back_01_l_ctrl\"),(Type=Control,Name=\"hair_strand_back_02_l_ctrl\"),(Type=Control,Name=\"hair_strand_back_03_l_ctrl\"),(Type=Control,Name=\"hair_strand_back_04_l_ctrl\"),(Type=Control,Name=\"hair_strand_back_05_l_ctrl\"),(Type=Cont ...(truncated)"
    Primary Axis=(X=1,Y=0,Z=0)
    Samples Per Segment=4
    Secondary Spline Direction=(X=0,Y=5,Z=0)
    Spline Mode=BSpline
    Stretch=false
    Up Axis=(X=0,Y=1,Z=0)
   )
   &RigUnit_PositionConstraintLocalSpaceOffset(
    Child=(Type=Null,Name="hair_strand_back_04_r_ctrl_null")
    Filter=(bX=True,bY=True,bZ=True)
    Parents=((Item=(Type=Control,Name="hair_strand_back_03_r_ctrl"),Weight=0.75),(Item=(Type=Control,Name="hair_strand_back_07_r_ctrl"),Weight=0.25))
    Weight=1
    bMaintainOffset=false
    ExecutePin=@SequenceAggregate.B
   )
   > &RigUnit_PositionConstraintLocalSpaceOffset(
    Child=(Type=Null,Name="hair_strand_back_05_r_ctrl_null")
    Filter=(bX=True,bY=True,bZ=True)
    Parents=((Item=(Type=Control,Name="hair_strand_back_03_r_ctrl"),Weight=0.525),(Item=(Type=Control,Name="hair_strand_back_07_r_ctrl"),Weight=0.475))
    Weight=1
    bMaintainOffset=false
   )
   > &RigUnit_PositionConstraintLocalSpaceOffset(
    Child=(Type=Null,Name="hair_strand_back_06_r_ctrl_null")
    Filter=(bX=True,bY=True,bZ=True)
    Parents=((Item=(Type=Control,Name="hair_strand_back_03_r_ctrl"),Weight=0.25),(Item=(Type=Control,Name="hair_strand_back_07_r_ctrl"),Weight=0.75))
    Weight=1
    bMaintainOffset=false
   )
   > &RigUnit_PositionConstraintLocalSpaceOffset(
    Child=(Type=Null,Name="hair_strand_back_07_r_ctrl_null")
    Filter=(bX=True,bY=True,bZ=True)
    Parents=((Item=(Type=Bone,Name="spine_05"),Weight=0.8),(Item=(Type=Bone,Name="clavicle_r"),Weight=0.2))
    Weight=1
    bMaintainOffset=True
   )
   > SplineIK_1 = RigVMFunctionReferenceNode(
    Bones="((Type=Bone,Name=\"hair_strand_back_01_r\"),(Type=Bone,Name=\"hair_strand_back_02_r\"),(Type=Bone,Name=\"hair_strand_back_03_r\"),(Type=Bone,Name=\"hair_strand_back_04_r\"),(Type=Bone,Name=\"hair_strand_back_05_r\"),(Type=Bone,Name=\"hair_strand_back_06_r\"),(Type=Bon ...(truncated)"
    Controls="((Type=Control,Name=\"hair_strand_back_01_r_ctrl\"),(Type=Control,Name=\"hair_strand_back_02_r_ctrl\"),(Type=Control,Name=\"hair_strand_back_03_r_ctrl\"),(Type=Control,Name=\"hair_strand_back_04_r_ctrl\"),(Type=Control,Name=\"hair_strand_back_05_r_ctrl\"),(Type=Cont ...(truncated)"
    Primary Axis=(X=1,Y=0,Z=0)
    Samples Per Segment=4
    Secondary Spline Direction=(X=0,Y=5,Z=0)
    Spline Mode=BSpline
    Stretch=false
    Up Axis=(X=0,Y=1,Z=0)
   )
  ]}
]
hierarchy [
 root {loc=<0 0 0>, rot=<0 0 0 1>}
 pelvis {parent=root, loc=<0 -0.5301 98.0571>, rot=<0.0127 0.707 -0.0127 -0.707>}
 spine_01 {parent=pelvis, loc=<6.1208 -0.9328 0>, rot=<0 0 0.034 -0.9994>}
 spine_02 {parent=spine_01, loc=<6.5996 -0.2095 0>, rot=<0 0 0.0238 -0.9997>}
 spine_03 {parent=spine_02, loc=<6.4914 -0.1237 0>, rot=<0 0 0.0149 -0.9999>}
 spine_04 {parent=spine_03, loc=<6.9598 0.0794 0>, rot=<0 0 -0.0132 -0.9999>}
 spine_05 {parent=spine_04, loc=<6.9987 0 0>, rot=<0 0 -0.0015 -1>}
 clavicle_l {parent=spine_05, loc=<5.4831 -0.1053 -3.2946>, rot=<0.0465 -0.7289 -0.009 -0.6829>}
 upperarm_l {parent=clavicle_l, loc=<9.8658 -1.2136 -0.4105>, rot=<0.0164 -0.3245 -0.0269 -0.9454>}
 neck_01 {parent=spine_05, loc=<11.3138 0.1529 0>, rot=<0 0 -0.2003 -0.9797>}
 neck_02 {parent=neck_01, loc=<4.5627 0 0>, rot=<0 0 0.0504 -0.9987>}
 head {parent=neck_02, loc=<5.3914 0 0>, rot=<0 0 0.115 -0.9934>}
 dyn_earring_l {parent=head, loc=<0.3535 1.0888 -6.0324>, rot=<0.0079 -0.7071 -0.7071 0.0089>}
 dyn_earring_r {parent=head, loc=<0.3538 1.0888 6.0324>, rot=<0.0079 -0.7071 -0.7071 0.0089>}
 hair_strand_side_mstr_l {parent=head, loc=<3.9063 -2.6761 -6.557>, rot=<-0.0043 0.0354 0.9971 -0.0678>}
 hair_strand_side_01_l {parent=hair_strand_side_mstr_l, loc=<0 0 0>, rot=<0 0 0 1>}
 hair_strand_side_02_l {parent=hair_strand_side_01_l, loc=<5.5 0 0>, rot=<-0.1834 -0.0102 0.2202 -0.958>}
 hair_strand_side_03_l {parent=hair_strand_side_02_l, loc=<5.5 0 0>, rot=<0.1134 0.1464 0.1217 -0.9751>}
 hair_strand_side_04_l {parent=hair_strand_side_03_l, loc=<5.5 0 0>, rot=<0.0217 -0.0177 0.0206 -0.9994>}
 hair_strand_side_05_l {parent=hair_strand_side_04_l, loc=<5.5 0 0>, rot=<0.012 -0.0597 -0.025 -0.9978>}
 hair_strand_side_mstr_r {parent=head, loc=<3.8558 -2.3079 6.5305>, rot=<0.0042 0.0397 0.9974 -0.0601>}
 hair_strand_side_01_r {parent=hair_strand_side_mstr_r, loc=<0 0 0>, rot=<0 0 0 1>}
 hair_strand_side_02_r {parent=hair_strand_side_01_r, loc=<5.4996 0.003 0.0424>, rot=<-0.1888 -0.0369 0.2522 -0.9484>}
 hair_strand_side_03_r {parent=hair_strand_side_02_r, loc=<5.3041 0.5924 1.328>, rot=<0.112 0.1527 0.0796 -0.9787>}
 hair_strand_side_04_r {parent=hair_strand_side_03_r, loc=<5.4344 -0.1508 -0.8334>, rot=<0.0143 -0.0103 -0.0324 -0.9993>}
 hair_strand_side_05_r {parent=hair_strand_side_04_r, loc=<5.4862 -0.0538 -0.3898>, rot=<0.028 0.0294 -0.0341 -0.9986>}
 hair_strand_back_mstr_l {parent=head, loc=<11.9857 -7.6352 -3.094>, rot=<0.5474 -0.7701 -0.3001 -0.1313>}
 hair_strand_back_01_l {parent=hair_strand_back_mstr_l, loc=<0 0 0>, rot=<0 0 0 1>}
 hair_strand_back_02_l {parent=hair_strand_back_01_l, loc=<5.5 0 0>, rot=<0.0503 -0.0163 -0.4349 -0.8989>}
 hair_strand_back_03_l {parent=hair_strand_back_02_l, loc=<5.5 0 0>, rot=<-0.1496 0.0805 -0.2031 -0.9643>}
 hair_strand_back_04_l {parent=hair_strand_back_03_l, loc=<5.5 0 0>, rot=<0.1013 0.0492 -0.0132 -0.9935>}
 hair_strand_back_05_l {parent=hair_strand_back_04_l, loc=<5.5 0 0>, rot=<-0.1688 0.034 0.0527 -0.9837>}
 hair_strand_back_06_l {parent=hair_strand_back_05_l, loc=<5.5 0 0>, rot=<0.1504 -0.0508 0.0687 -0.9849>}
 hair_strand_back_07_l {parent=hair_strand_back_06_l, loc=<7.25 0 0>, rot=<0 0 0 1>}
 hair_strand_back_mstr_r {parent=head, loc=<11.9853 -7.6352 3.094>, rot=<0.5474 -0.7701 0.3001 0.1313>}
 hair_strand_back_01_r {parent=hair_strand_back_mstr_r, loc=<0 0 0>, rot=<0 0 0 1>}
 hair_strand_back_02_r {parent=hair_strand_back_01_r, loc=<5.5001 0 0>, rot=<-0.0503 0.0163 -0.4349 -0.8989>}
 hair_strand_back_03_r {parent=hair_strand_back_02_r, loc=<5.5 0 0>, rot=<0.1496 -0.0805 -0.2031 -0.9643>}
 hair_strand_back_04_r {parent=hair_strand_back_03_r, loc=<5.4998 0 0>, rot=<-0.1013 -0.0492 -0.0132 -0.9935>}
 hair_strand_back_05_r {parent=hair_strand_back_04_r, loc=<5.4998 0 0>, rot=<0.1688 -0.0339 0.0527 -0.9837>}
 hair_strand_back_06_r {parent=hair_strand_back_05_r, loc=<5.5002 0 0>, rot=<-0.1504 0.0508 0.0687 -0.9849>}
 hair_strand_back_07_r {parent=hair_strand_back_06_r, loc=<7.25 0 0>, rot=<0 0 0 1>}
 dyn_ponytail_mstr {parent=head, loc=<11.4435 -7.9638 0.0018>, rot=<0.7355 -0.6775 0 0>}
 dyn_ponytail_base {parent=dyn_ponytail_mstr, loc=<0 0 0>, rot=<0 0 0 1>}
 dyn_ponytail_01 {parent=dyn_ponytail_base, loc=<4.8313 0 0>, rot=<0.0257 0.0998 -0.6235 -0.775>}
 dyn_ponytail_02 {parent=dyn_ponytail_01, loc=<6.9684 0 0>, rot=<0.0352 -0.0211 -0.2544 -0.9662>}
 dyn_ponytail_03 {parent=dyn_ponytail_02, loc=<7.0615 0 0>, rot=<0.1352 -0.1748 -0.163 -0.9616>}
 dyn_ponytail_04 {parent=dyn_ponytail_03, loc=<6.883 0 0>, rot=<-0.4206 -0.2473 -0.2837 -0.8255>}
 dyn_ponytail_05 {parent=dyn_ponytail_04, loc=<6.9592 0 0>, rot=<-0.7252 -0.0931 -0.2063 -0.6503>}
 dyn_ponytail_tip_mstr {parent=dyn_ponytail_05, loc=<6.3146 0 0>, rot=<-0.0076 0.0075 -0.2211 -0.9752>}
 dyn_ponytail_tip_01 {parent=dyn_ponytail_tip_mstr, loc=<0 0 0>, rot=<0 0 -0.0203 -0.9998>}
 dyn_ponytail_tip_scale_01 {parent=dyn_ponytail_tip_01, loc=<0 0 0>, rot=<0 0 0 1>}
 dyn_ponytail_tip_02 {parent=dyn_ponytail_tip_01, loc=<3.8706 0 0>, rot=<-0.0004 0.0068 -0.1448 -0.9894>}
 dyn_ponytail_tip_scale_02 {parent=dyn_ponytail_tip_02, loc=<0 0 0>, rot=<0 0 0 1>, scale=<1 1 0.6338>}
 dyn_ponytail_tip_03 {parent=dyn_ponytail_tip_02, loc=<3.8496 0 0>, rot=<-0.0673 0.0577 -0.1465 -0.9852>}
 dyn_ponytail_tip_scale_03 {parent=dyn_ponytail_tip_03, loc=<0 0 0>, rot=<0 0 0 1>}
 dyn_ponytail_tip_04 {parent=dyn_ponytail_tip_03, loc=<3.8574 0 0>, rot=<-0.0394 0.1195 -0.0453 -0.991>}
 dyn_ponytail_tip_scale_04 {parent=dyn_ponytail_tip_04, loc=<0 0 0>, rot=<0 0 0 1>, scale=<1 1 0.6338>}
 dyn_ponytail_tip_05 {parent=dyn_ponytail_tip_04, loc=<3.8904 0 0>, rot=<0 0 0 1>}
 dyn_ponytail_tip_scale_05 {parent=dyn_ponytail_tip_05, loc=<0 0 0>, rot=<0 0 0 1>, scale=<1 1 0.6338>}
 dyn_hair_strand_01 {parent=dyn_ponytail_base, loc=<-0.7829 -3.7632 0.8109>, rot=<-0.1816 0.3665 0.5427 -0.7336>}
 dyn_hair_strand_02 {parent=dyn_ponytail_base, loc=<-1.0864 -3.5588 0.4231>, rot=<0.3028 -0.9159 0.0683 -0.2544>}
 dyn_hair_strand_03 {parent=dyn_ponytail_base, loc=<4.0385 -3.356 4.5535>, rot=<0.1375 0.3456 -0.244 -0.8956>}
 dyn_hair_strand_04 {parent=dyn_ponytail_base, loc=<3.4796 -6.1852 -0.3526>, rot=<0.0029 0.0129 0.0816 -0.9966>}
 dyn_hair_strand_05 {parent=dyn_ponytail_base, loc=<5.1132 -4.5986 -3.6762>, rot=<-0.2743 0.0144 -0.136 -0.9519>}
 dyn_hair_strand_06 {parent=dyn_ponytail_base, loc=<0.0528 -1.9865 -3.1112>, rot=<-0.3819 -0.2822 0.5314 -0.7015>}
 hair_strand_back_01_l_ctrl_null {parent=head, type=Null, loc=<11.9857 -7.6352 -3.094>, rot=<0.5474 -0.7701 -0.3001 -0.1313>}
 hair_strand_back_01_l_ctrl {controlType=EULER_TRANSFORM, parent=hair_strand_back_01_l_ctrl_null, type=Control, loc=<0 0 0>, rot=<0 0 0 1>}
 hair_strand_back_02_l_ctrl_null {parent=head, type=Null, loc=<9.9715 -11.8389 -6.0132>, rot=<0.1687 -0.9174 -0.3567 0.0526>}
 hair_strand_back_02_l_ctrl {controlType=EULER_TRANSFORM, parent=hair_strand_back_02_l_ctrl_null, type=Control, loc=<0 0 0>, rot=<0 0 0 1>}
 hair_strand_back_03_l_ctrl_null {parent=head, type=Null, loc=<4.815 -13.7479 -6.144>, rot=<-0.0444 -0.9765 -0.2096 0.0241>}
 hair_strand_back_03_l_ctrl {controlType=EULER_TRANSFORM, parent=hair_strand_back_03_l_ctrl_null, type=Control, loc=<0 0 0>, rot=<0 0 0 1>}
 hair_strand_back_01_r_ctrl_null {parent=head, type=Null, loc=<11.9853 -7.6352 3.094>, rot=<0.5474 -0.7701 0.3001 0.1313>}
 hair_strand_back_01_r_ctrl {controlType=EULER_TRANSFORM, parent=hair_strand_back_01_r_ctrl_null, type=Control, loc=<0 0 0>, rot=<0 0 0 1>}
 hair_strand_back_02_r_ctrl_null {parent=head, type=Null, loc=<9.971 -11.839 6.0132>, rot=<0.1687 -0.9174 0.3567 -0.0526>}
 hair_strand_back_02_r_ctrl {controlType=EULER_TRANSFORM, parent=hair_strand_back_02_r_ctrl_null, type=Control, loc=<0 0 0>, rot=<0 0 0 1>}
 hair_strand_back_03_r_ctrl_null {parent=head, type=Null, loc=<4.8146 -13.7479 6.144>, rot=<-0.0444 -0.9765 0.2096 -0.0241>}
 hair_strand_back_03_r_ctrl {controlType=EULER_TRANSFORM, parent=hair_strand_back_03_r_ctrl_null, type=Control, loc=<0 0 0>, rot=<0 0 0 1>}
 dfrm_collar_l {parent=neck_01, loc=<4.5999 1.9086 -5.6788>, rot=<0 0 -0.0049 -1>}
 dfrm_collar_r {parent=neck_01, loc=<4.5999 1.9086 5.679>, rot=<0 0 -0.0049 -1>}
 clavicle_r {parent=spine_05, loc=<5.4831 -0.1053 3.2946>, rot=<0.7289 0.0465 -0.6829 0.009>}
 upperarm_r {parent=clavicle_r, loc=<-9.8658 1.2136 0.4105>, rot=<0.0164 -0.3245 -0.0269 -0.9454>}
 hair_strand_back_04_l_ctrl_null {parent=spine_05, type=Null, loc=<20.9602 -9.79 -5.7825>, rot=<0.0356 0.9514 0.3054 0.0148>}
 hair_strand_back_04_l_ctrl {controlType=EULER_TRANSFORM, parent=hair_strand_back_04_l_ctrl_null, type=Control, loc=<0 0 0>, rot=<0 0 0 1>}
 hair_strand_back_05_l_ctrl_null {parent=spine_05, type=Null, loc=<15.4765 -9.3676 -5.8176>, rot=<-0.0022 0.9888 0.1378 0.0569>}
 hair_strand_back_05_l_ctrl {controlType=EULER_TRANSFORM, parent=hair_strand_back_05_l_ctrl_null, type=Control, loc=<0 0 0>, rot=<0 0 0 1>}
 hair_strand_back_06_l_ctrl_null {parent=spine_05, type=Null, loc=<10.0122 -9.3054 -6.4401>, rot=<-0.0857 0.9559 0.2804 0.015>}
 hair_strand_back_06_l_ctrl {controlType=EULER_TRANSFORM, parent=hair_strand_back_06_l_ctrl_null, type=Control, loc=<0 0 0>, rot=<0 0 0 1>}
 hair_strand_back_07_l_ctrl_null {parent=spine_05, type=Null, loc=<2.8719 -10.4319 -6.9958>, rot=<-0.0857 0.9559 0.2804 0.015>}
 hair_strand_back_07_l_ctrl {controlType=EULER_TRANSFORM, parent=hair_strand_back_07_l_ctrl_null, type=Control, loc=<0 0 0>, rot=<0 0 0 1>}
 hair_strand_back_04_r_ctrl_null {parent=spine_05, type=Null, loc=<20.9599 -9.7901 5.7825>, rot=<0.0356 0.9514 -0.3054 -0.0148>}
 hair_strand_back_04_r_ctrl {controlType=EULER_TRANSFORM, parent=hair_strand_back_04_r_ctrl_null, type=Control, loc=<0 0 0>, rot=<0 0 0 1>}
 hair_strand_back_05_r_ctrl_null {parent=spine_05, type=Null, loc=<15.4765 -9.3676 5.8176>, rot=<-0.0022 0.9888 -0.1378 -0.0569>}
 hair_strand_back_05_r_ctrl {controlType=EULER_TRANSFORM, parent=hair_strand_back_05_r_ctrl_null, type=Control, loc=<0 0 0>, rot=<0 0 0 1>}
 hair_strand_back_06_r_ctrl_null {parent=spine_05, type=Null, loc=<10.0119 -9.3054 6.4401>, rot=<-0.0857 0.9559 -0.2804 -0.015>}
 hair_strand_back_06_r_ctrl {controlType=EULER_TRANSFORM, parent=hair_strand_back_06_r_ctrl_null, type=Control, loc=<0 0 0>, rot=<0 0 0 1>}
 hair_strand_back_07_r_ctrl_null {parent=spine_05, type=Null, loc=<2.8715 -10.4319 6.9958>, rot=<-0.0857 0.9559 -0.2804 -0.015>}
 hair_strand_back_07_r_ctrl {controlType=EULER_TRANSFORM, parent=hair_strand_back_07_r_ctrl_null, type=Control, loc=<0 0 0>, rot=<0 0 0 1>}
]
