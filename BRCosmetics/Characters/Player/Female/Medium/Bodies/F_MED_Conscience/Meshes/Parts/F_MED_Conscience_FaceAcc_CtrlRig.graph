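# Control Rig for the female Conscience character's face accessory, driving the ponytail, its tip and both side hair strands.
# Construction snapshots bind and corrective control poses into variables.
# Spherical pose readers on head, neck, left clavicle and right upper arm weight those poses, plus tip scale and right collar tweaks.
# Constrained controls feed four B-spline IK chains, while front-end mode adds side-strand adjustments, interpolated for emotes.
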ControlRigBlueprint /BRCosmetics/Characters/Player/Female/Medium/Bodies/F_MED_Conscience/Meshes/Parts/F_MED_Conscience_FaceAcc_CtrlRig
graphs [
 ContainedGraph {nodes [
   n1 = RigVMFunctionEntryNode()
   SequenceAggregate = &RigVMFunction_Sequence(ExecuteContext=@n1.ExecuteContext)
   SequenceAggregate_1 = &RigVMFunction_Sequence(ExecuteContext=@SequenceAggregate.B)
   SequenceAggregate_2 = &RigVMFunction_Sequence(ExecuteContext=@SequenceAggregate_1.B)
   RigVMFunctionReturnNode(A=@SequenceAggregate.A, B=@SequenceAggregate_1.A, C=@SequenceAggregate_2.A, D=@SequenceAggregate_2.B)
  ]}
 ContainedGraph {nodes [
   n1 = RigVMFunctionEntryNode()
   SequenceAggregate_1 = &RigVMFunction_Sequence(ExecuteContext=@n1.ExecuteContext)
   SequenceAggregate_1_1 = &RigVMFunction_Sequence(ExecuteContext=@SequenceAggregate_1.B)
   RigVMFunctionReturnNode(A=@SequenceAggregate_1.A, B=@SequenceAggregate_1_1.A, C=@SequenceAggregate_1_1.B)
  ]}
 ContainedGraph {nodes [
   n1 = RigVMFunctionEntryNode()
   SequenceAggregate_3 = &RigVMFunction_Sequence(ExecuteContext=@n1.ExecuteContext)
   SequenceAggregate_3_1 = &RigVMFunction_Sequence(ExecuteContext=@SequenceAggregate_3.B)
   SequenceAggregate_3_2 = &RigVMFunction_Sequence(ExecuteContext=@SequenceAggregate_3_1.B)
   SequenceAggregate_3_3 = &RigVMFunction_Sequence(ExecuteContext=@SequenceAggregate_3_2.B)
   RigVMFunctionReturnNode(A=@SequenceAggregate_3.A, B=@SequenceAggregate_3_1.A, C=@SequenceAggregate_3_2.A, D=@SequenceAggregate_3_3.A, E=@SequenceAggregate_3_3.B)
  ]}
 ContainedGraph {nodes [
   n1 = RigVMFunctionEntryNode()
   SequenceAggregate_4 = &RigVMFunction_Sequence(ExecuteContext=@n1.ExecuteContext)
   SequenceAggregate_4_1 = &RigVMFunction_Sequence(ExecuteContext=@SequenceAggregate_4.B)
   SequenceAggregate_4_2 = &RigVMFunction_Sequence(ExecuteContext=@SequenceAggregate_4_1.B)
   RigVMFunctionReturnNode(A=@SequenceAggregate_4.A, B=@SequenceAggregate_4_1.A, C=@SequenceAggregate_4_2.A, D=@SequenceAggregate_4_2.B)
  ]}
 ContainedGraph {nodes [
   n1 = RigVMFunctionEntryNode()
   SequenceAggregate_6 = &RigVMFunction_Sequence(ExecuteContext=@n1.ExecuteContext)
   SequenceAggregate_6_1 = &RigVMFunction_Sequence(ExecuteContext=@SequenceAggregate_6.B)
   SequenceAggregate_6_2 = &RigVMFunction_Sequence(ExecuteContext=@SequenceAggregate_6_1.B)
   SequenceAggregate_6_3 = &RigVMFunction_Sequence(ExecuteContext=@SequenceAggregate_6_2.B)
   SequenceAggregate_6_4 = &RigVMFunction_Sequence(ExecuteContext=@SequenceAggregate_6_3.B)
   RigVMFunctionReturnNode(
    A=@SequenceAggregate_6.A
    B=@SequenceAggregate_6_1.A
    C=@SequenceAggregate_6_2.A
    D=@SequenceAggregate_6_3.A
    E=@SequenceAggregate_6_4.A
    F=@SequenceAggregate_6_4.B
   )
  ]}
 ContainedGraph {nodes [
   n1 = RigVMFunctionEntryNode()
   SequenceAggregate_5 = &RigVMFunction_Sequence(ExecuteContext=@n1.ExecuteContext)
   SequenceAggregate_5_1 = &RigVMFunction_Sequence(ExecuteContext=@SequenceAggregate_5.B)
   RigVMFunctionReturnNode(A=@SequenceAggregate_5.A, B=@SequenceAggregate_5_1.A, C=@SequenceAggregate_5_1.B)
  ]}
 ContainedGraph {nodes [
   n1 = RigVMFunctionEntryNode()
   SequenceAggregate_7 = &RigVMFunction_Sequence(ExecuteContext=@n1.ExecuteContext)
   SequenceAggregate_7_1 = &RigVMFunction_Sequence(ExecuteContext=@SequenceAggregate_7.B)
   SequenceAggregate_7_2 = &RigVMFunction_Sequence(ExecuteContext=@SequenceAggregate_7_1.B)
   RigVMFunctionReturnNode(A=@SequenceAggregate_7.A, B=@SequenceAggregate_7_1.A, C=@SequenceAggregate_7_2.A, D=@SequenceAggregate_7_2.B)
  ]}
 ContainedGraph {nodes [
   n1 = RigVMFunctionEntryNode()
   SequenceAggregate_1_1 = &RigVMFunction_Sequence(ExecuteContext=@n1.ExecuteContext)
   SequenceAggregate_1_1_1 = &RigVMFunction_Sequence(ExecuteContext=@SequenceAggregate_1_1.B)
   SequenceAggregate_1_1_2 = &RigVMFunction_Sequence(ExecuteContext=@SequenceAggregate_1_1_1.B)
   SequenceAggregate_1_1_3 = &RigVMFunction_Sequence(ExecuteContext=@SequenceAggregate_1_1_2.B)
   RigVMFunctionReturnNode(
    A=@SequenceAggregate_1_1.A
    B=@SequenceAggregate_1_1_1.A
    C=@SequenceAggregate_1_1_2.A
    D=@SequenceAggregate_1_1_3.A
    E=@SequenceAggregate_1_1_3.B
   )
  ]}
 ContainedGraph {nodes [
   n1 = RigVMFunctionEntryNode()
   SequenceAggregate_2 = &RigVMFunction_Sequence(ExecuteContext=@n1.ExecuteContext)
   SequenceAggregate_2_1 = &RigVMFunction_Sequence(ExecuteContext=@SequenceAggregate_2.B)
   SequenceAggregate_2_2 = &RigVMFunction_Sequence(ExecuteContext=@SequenceAggregate_2_1.B)
   SequenceAggregate_2_3 = &RigVMFunction_Sequence(ExecuteContext=@SequenceAggregate_2_2.B)
   SequenceAggregate_2_4 = &RigVMFunction_Sequence(ExecuteContext=@SequenceAggregate_2_3.B)
   RigVMFunctionReturnNode(
    A=@SequenceAggregate_2.A
    B=@SequenceAggregate_2_1.A
    C=@SequenceAggregate_2_2.A
    D=@SequenceAggregate_2_3.A
    E=@SequenceAggregate_2_4.A
    F=@SequenceAggregate_2_4.B
   )
  ]}
 ContainedGraph {nodes [
   n1 = RigVMFunctionEntryNode()
   SequenceAggregate = &RigVMFunction_Sequence(ExecuteContext=@n1.ExecuteContext)
   SequenceAggregate_1 = &RigVMFunction_Sequence(ExecuteContext=@SequenceAggregate.B)
   SequenceAggregate_2 = &RigVMFunction_Sequence(ExecuteContext=@SequenceAggregate_1.B)
   RigVMFunctionReturnNode(A=@SequenceAggregate.A, B=@SequenceAggregate_1.A, C=@SequenceAggregate_2.A, D=@SequenceAggregate_2.B)
  ]}
 ContainedGraph {nodes [
   n1 = RigVMFunctionEntryNode()
   SequenceAggregate_1 = &RigVMFunction_Sequence(ExecuteContext=@n1.ExecuteContext)
   SequenceAggregate_1_1 = &RigVMFunction_Sequence(ExecuteContext=@SequenceAggregate_1.B)
   SequenceAggregate_1_2 = &RigVMFunction_Sequence(ExecuteContext=@SequenceAggregate_1_1.B)
   RigVMFunctionReturnNode(A=@SequenceAggregate_1.A, B=@SequenceAggregate_1_1.A, C=@SequenceAggregate_1_2.A, D=@SequenceAggregate_1_2.B)
  ]}
 ContainedGraph {nodes [
   n1 = RigVMFunctionEntryNode()
   SequenceAggregate = &RigVMFunction_Sequence(ExecuteContext=@n1.ExecuteContext)
   Subtract = &RigVMFunction_MathDoubleSub(A=1, B=@n1.pose_alpha)
   /Script/ControlRig.RigUnit_HierarchySetPoseItemArray(ElementType=All, ItemsToSet=(), Pose=@n1.bind_pose, ExecutePin=@SequenceAggregate.A, Weight=@Subtract.Result) {is=struct RigVMUnitNode}
   &RigUnit_HierarchySetPoseItemArray(ElementType=All, ItemsToSet=(), Weight=@n1.pose_alpha, Pose=@n1.target_pose, ExecutePin=@SequenceAggregate.B)
   > RigVMFunctionReturnNode()
  ]}
 RigVMFunctionLibrary {nodes [set_pose_func = RigVMCollapseNode(ExecuteContext=(), bind_pose=(HierarchyTopologyVersion=-1,PoseHash=-1), target_pose=(HierarchyTopologyVersion=-1,PoseHash=-1))]}
 RigVMModel {nodes [
   RigUnit_BeginExecution = &RigUnit_BeginExecution()
   SequenceAggregate_1 = RigVMAggregateNode(ExecuteContext=@RigUnit_BeginExecution.ExecutePin)
   SequenceAggregate_2 = &RigVMFunction_Sequence(ExecuteContext=@SequenceAggregate_1.C)
   SequenceAggregate_4 = RigVMAggregateNode(ExecuteContext=@SequenceAggregate_2.A)
   set_pose_func_2 = RigVMFunctionReferenceNode(
    ExecuteContext=@SequenceAggregate_4.A
    pose_alpha=RigVMVariableNode() {variable=spr_upperarm_r_up_pistol_alpha}.Value
    bind_pose=RigVMVariableNode() {variable=ponytail_spline_ik_bind_pose}.Value
    target_pose=RigVMVariableNode() {variable=ponytail_spline_ik_up_right_arm}.Value
   )
   SequenceAggregate_5 = RigVMAggregateNode(ExecuteContext=@SequenceAggregate_4.C)
   n1 = RigVMVariableNode() {variable=spr_upperarm_r_up_pistol_alpha}
   Subtract = &RigVMFunction_MathDoubleSub(A=1, B=@n1.Value)
   n2 = RigVMVariableNode() {variable=spr_neck_down_alpha}
   &RigUnit_HierarchySetPoseItemArray(
    ElementType=All
    ItemsToSet=()
    Weight=&RigVMFunction_MathDoubleClamp(Value=&RigVMFunction_MathDoubleSub(A=@Subtract.Result, B=@n2.Value).Result).Result
    ExecutePin=@SequenceAggregate_5.A
    Pose=RigVMVariableNode() {variable=right_strand_spline_ik_bind_pose}.Value
   )
   &RigUnit_HierarchySetPoseItemArray(
    ElementType=All
    ItemsToSet=()
    Pose=RigVMVariableNode() {variable=right_strand_spline_ik_neck_01_down}.Value
    Weight=@n2.Value
    ExecutePin=@SequenceAggregate_5.C
   )
   &RigUnit_HierarchySetPoseItemArray(
    ElementType=All
    ItemsToSet=()
    Weight=@n1.Value
    Pose=RigVMVariableNode() {variable=right_strand_spline_ik_up_right_pstl}.Value
    ExecutePin=@SequenceAggregate_5.B
   )
   set_pose_func_3 = RigVMFunctionReferenceNode(
    bind_pose=RigVMVariableNode() {variable=ponytail_tip_spline_ik_bind_pose}.Value
    pose_alpha=RigVMVariableNode() {variable=spr_upperarm_r_up_pistol_alpha}.Value
    target_pose=RigVMVariableNode() {variable=ponytail_tip_spline_ik_up_right_pstl}.Value
    ExecuteContext=@SequenceAggregate_4.B
   )
   SequenceAggregate_6 = RigVMAggregateNode(ExecuteContext=@SequenceAggregate_4.D)
   n3 = RigVMVariableNode() {variable=spr_clavicle_l_up_alpha}
   n4 = &RigVMFunction_MathDoubleMul(
    A=&RigVMFunction_MathDoubleSub(A=1, B=RigVMVariableNode() {variable=spr_look_left_alpha}.Value).Result
    B=&RigVMFunction_MathDoubleSub(A=1, B=@n3.Value).Result
   )
   n5 = RigVMVariableNode() {variable=spr_neck_down_alpha}
   n6 = &RigVMFunction_MathDoubleSub(A=1, B=@n5.Value)
   Clamp = &RigVMFunction_MathDoubleClamp(Value=&RigVMFunction_MathDoubleSub(A=&RigVMFunction_MathDoubleSub(B=@n4.Result, A=@n6.Result).Result, B=@n3.Value).Result)
   &RigUnit_HierarchySetPoseItemArray(
    ElementType=All
    ItemsToSet=()
    Weight=@Clamp.Result
    ExecutePin=@SequenceAggregate_6.A
    Pose=RigVMVariableNode() {variable=left_strand_spline_ik_bind_pose}.Value
   )
   &RigUnit_HierarchySetPoseItemArray(
    ElementType=All
    ItemsToSet=()
    Pose=RigVMVariableNode() {variable=left_strand_spline_ik_neck_01_down}.Value
    ExecutePin=@SequenceAggregate_6.B
    Weight=@n5.Value
   )
   Branch_RigVMUnitNode = &RigVMFunction_ControlFlowBranch(Condition=RigVMVariableNode() {variable=is_front_end}.Value, ExecuteContext=@SequenceAggregate_6.F)
   SequenceAggregate_7 = RigVMAggregateNode(ExecuteContext=@Branch_RigVMUnitNode.True)
   Cast = &RigVMFunction_MathBoolToFloat(Value=RigVMVariableNode() {variable=is_playing_emote}.Value)
   AlphaInterp = &RigVMFunction_AlphaInterp(InterpSpeedDecreasing=10, InterpSpeedIncreasing=10, Scale=1, bClampResult=False, Value=@Cast.Result)
   n7 = &RigVMFunction_MathDoubleSub(A=1, B=@AlphaInterp.Result)
   &RigUnit_HierarchySetPoseItemArray(
    ElementType=All
    ItemsToSet=()
    ExecutePin=@SequenceAggregate_7.A
    Pose=RigVMVariableNode() {variable=left_strand_FE_pose_adjustment}.Value
    Weight=@n7.Result
   )
   &RigUnit_HierarchySetPoseItemArray(
    ElementType=All
    ItemsToSet=()
    Pose=RigVMVariableNode() {variable=right_strand_FE_pose_adjustment}.Value
    ExecutePin=@SequenceAggregate_7.B
    Weight=@n7.Result
   )
   &RigUnit_HierarchySetPoseItemArray(
    ElementType=All
    ItemsToSet=()
    Pose=RigVMVariableNode() {variable=right_strand_FE_pose_emote_adjustment}.Value
    ExecutePin=@SequenceAggregate_7.D
    Weight=@AlphaInterp.Result
   )
   &RigUnit_HierarchySetPoseItemArray(
    ElementType=All
    ItemsToSet=()
    Pose=RigVMVariableNode() {variable=left_strand_FE_pose_emote_adjustment}.Value
    ExecutePin=@SequenceAggregate_7.C
    Weight=@AlphaInterp.Result
   )
   &RigUnit_HierarchySetPoseItemArray(
    ElementType=All
    ItemsToSet=()
    Weight=@n4.Result
    Pose=RigVMVariableNode() {variable=left_strand_spline_ik_look_left_pose}.Value
    ExecutePin=@SequenceAggregate_6.C
   )
   &RigUnit_HierarchySetPoseItemArray(
    ElementType=All
    ItemsToSet=()
    Pose=RigVMVariableNode() {variable=left_strand_spline_ik_clav_l_up}.Value
    ExecutePin=@SequenceAggregate_6.D
    Weight=@n3.Value
   )
   &RigUnit_HierarchySetPoseItemArray(
    ElementType=All
    ItemsToSet=()
    Weight=&RigVMFunction_MathDoubleMul(A=RigVMVariableNode() {variable=spr_look_up_alpha}.Value, B=@n3.Value).Result
    Pose=RigVMVariableNode() {variable=left_strand_spline_ik_clav_up_head_up}.Value
    ExecutePin=@SequenceAggregate_6.E
   )
   n8 = RigVMVariableNode() {variable=spr_upperarm_r_up_pistol_alpha}
   /Script/ControlRig.RigUnit_SetScale(
    Item=(Type=Bone,Name="dyn_ponytail_tip_scale_02")
    Scale=(X=1,Y=0.5,Z=0.5)
    Space=GlobalSpace
    bPropagateToChildren=True
    Weight=@n8.Value
    ExecutePin=@SequenceAggregate_2.B
   ) {is=struct RigVMUnitNode}
   > &RigUnit_SetScale(Item=(Type=Bone,Name="dyn_ponytail_tip_scale_01"), Scale=(X=1,Y=0.5,Z=0.5), Space=GlobalSpace, bPropagateToChildren=True, Weight=@n8.Value)
   > &RigUnit_ModifyTransforms(
    ItemToModify=((Item=(Type=Bone,Name="dfrm_collar_r"),Transform=(Rotation=(X=0,Y=0,Z=0,W=1),Translation=(X=0,Y=0,Z=-1.25),Scale3D=(X=1,Y=1,Z=1))))
    Weight=@n8.Value
   )
   SequenceAggregate = RigVMAggregateNode(ExecuteContext=@SequenceAggregate_1.B)
   /Script/ControlRig.RigUnit_PositionConstraintLocalSpaceOffset(
    Child=(Type=Null,Name="hair_strand_side_04_r_ctrl_null")
    Filter=(bX=True,bY=True,bZ=True)
    Parents=((Item=(Type=Control,Name="hair_strand_side_05_r_ctrl"),Weight=0.5),(Item=(Type=Control,Name="hair_strand_side_03_r_ctrl"),Weight=0.5))
    Weight=1
    bMaintainOffset=false
    ExecutePin=@SequenceAggregate.B
   ) {is=struct RigVMUnitNode}
   > &RigUnit_PositionConstraintLocalSpaceOffset(
    Child=(Type=Null,Name="hair_strand_side_02_r_ctrl_null")
    Filter=(bX=True,bY=True,bZ=True)
    Parents=((Item=(Type=Control,Name="hair_strand_side_01_r_ctrl"),Weight=0.5),(Item=(Type=Control,Name="hair_strand_side_03_r_ctrl"),Weight=0.5))
    Weight=1
    bMaintainOffset=false
   )
   > SplineIK_1 = RigVMFunctionReferenceNode(
    Bones=((Type=Bone,Name="hair_strand_side_01_r"),(Type=Bone,Name="hair_strand_side_02_r"),(Type=Bone,Name="hair_strand_side_03_r"),(Type=Bone,Name="hair_strand_side_04_r"),(Type=Bone,Name="hair_strand_side_05_r"))
    Controls=((Type=Control,Name="hair_strand_side_01_r_ctrl"),(Type=Control,Name="hair_strand_side_02_r_ctrl"),(Type=Control,Name="hair_strand_side_03_r_ctrl"),(Type=Control,Name="hair_strand_side_04_r_ctrl"),(Type=Control,Name="hair_strand_side_05_r_ctrl"))
    Primary Axis=(X=1,Y=0,Z=0)
    Samples Per Segment=8
    Secondary Spline Direction=(X=0,Y=5,Z=0)
    Spline Mode=BSpline
    Stretch=false
    Up Axis=(X=0,Y=1,Z=0)
    debug=false
   )
   &RigUnit_PositionConstraintLocalSpaceOffset(
    Child=(Type=Null,Name="dyn_ponytail_02_ctrl_null")
    Filter=(bX=True,bY=True,bZ=True)
    Parents=((Item=(Type=Control,Name="dyn_ponytail_01_ctrl"),Weight=0.6),(Item=(Type=Control,Name="dyn_ponytail_05_ctrl"),Weight=0.4))
    Weight=1
    bMaintainOffset=True
    ExecutePin=@SequenceAggregate.C
   )
   > &RigUnit_PositionConstraintLocalSpaceOffset(
    Child=(Type=Null,Name="dyn_ponytail_03_ctrl_null")
    Filter=(bX=True,bY=True,bZ=True)
    Parents=((Item=(Type=Control,Name="dyn_ponytail_01_ctrl"),Weight=0.5),(Item=(Type=Control,Name="dyn_ponytail_05_ctrl"),Weight=0.5))
    Weight=1
    bMaintainOffset=True
   )
   > &RigUnit_PositionConstraintLocalSpaceOffset(
    Child=(Type=Null,Name="dyn_ponytail_04_ctrl_null")
    Filter=(bX=True,bY=True,bZ=True)
    Parents=((Item=(Type=Control,Name="dyn_ponytail_01_ctrl"),Weight=0.25),(Item=(Type=Control,Name="dyn_ponytail_05_ctrl"),Weight=0.75))
    Weight=1
    bMaintainOffset=True
   )
   > SplineIK_2 = RigVMFunctionReferenceNode(
    Bones=((Type=Bone,Name="dyn_ponytail_01"),(Type=Bone,Name="dyn_ponytail_02"),(Type=Bone,Name="dyn_ponytail_03"),(Type=Bone,Name="dyn_ponytail_04"),(Type=Bone,Name="dyn_ponytail_05"))
    Controls=((Type=Control,Name="dyn_ponytail_01_ctrl"),(Type=Control,Name="dyn_ponytail_02_ctrl"),(Type=Control,Name="dyn_ponytail_03_ctrl"),(Type=Control,Name="dyn_ponytail_04_ctrl"),(Type=Control,Name="dyn_ponytail_05_ctrl"))
    Primary Axis=(X=1,Y=0,Z=0)
    Samples Per Segment=8
    Secondary Spline Direction=(X=0,Y=5,Z=0)
    Spline Mode=BSpline
    Stretch=true
    Up Axis=(X=0,Y=1,Z=0)
   )
   &RigUnit_ParentConstraint(
    AdvancedSettings=(RotationOrderForFilter=XZY)
    Child=(Type=Null,Name="dyn_ponytail_tip_04_ctrl_null")
    Parents=((Item=(Type=Control,Name="dyn_ponytail_tip_03_ctrl"),Weight=0.5),(Item=(Type=Control,Name="dyn_ponytail_tip_05_ctrl"),Weight=0.5))
    Weight=1
    ExecutePin=@SequenceAggregate.D
   )
   > &RigUnit_ParentConstraint(
    AdvancedSettings=(RotationOrderForFilter=XZY)
    Child=(Type=Null,Name="dyn_ponytail_tip_02_ctrl_null")
    Parents=((Item=(Type=Control,Name="dyn_ponytail_tip_01_ctrl"),Weight=0.5),(Item=(Type=Control,Name="dyn_ponytail_tip_03_ctrl"),Weight=0.5))
    Weight=1
   )
   > SplineIK_3 = RigVMFunctionReferenceNode(
    Bones=((Type=Bone,Name="dyn_ponytail_tip_01"),(Type=Bone,Name="dyn_ponytail_tip_02"),(Type=Bone,Name="dyn_ponytail_tip_03"),(Type=Bone,Name="dyn_ponytail_tip_04"),(Type=Bone,Name="dyn_ponytail_tip_05"))
    Controls=((Type=Control,Name="dyn_ponytail_tip_01_ctrl"),(Type=Control,Name="dyn_ponytail_tip_02_ctrl"),(Type=Control,Name="dyn_ponytail_tip_03_ctrl"),(Type=Control,Name="dyn_ponytail_tip_04_ctrl"),(Type=Control,Name="dyn_ponytail_tip_05_ctrl"))
    Primary Axis=(X=1,Y=0,Z=0)
    Samples Per Segment=8
    Secondary Spline Direction=(X=0,Y=5,Z=0)
    Spline Mode=BSpline
    Up Axis=(X=0,Y=1,Z=0)
   )
   &RigUnit_ParentConstraint(
    AdvancedSettings=(RotationOrderForFilter=XZY)
    Child=(Type=Null,Name="hair_strand_side_04_l_ctrl_null")
    Parents=((Item=(Type=Control,Name="hair_strand_side_05_l_ctrl"),Weight=0.5),(Item=(Type=Control,Name="hair_strand_side_05_l_ctrl"),Weight=0.5))
    Weight=1
    ExecutePin=@SequenceAggregate.A
   )
   > &RigUnit_ParentConstraint(
    AdvancedSettings=(RotationOrderForFilter=XZY)
    Child=(Type=Null,Name="hair_strand_side_02_l_ctrl_null")
    Parents=((Item=(Type=Control,Name="hair_strand_side_01_l_ctrl"),Weight=0.5),(Item=(Type=Control,Name="hair_strand_side_03_l_ctrl"),Weight=0.5))
    Weight=1
   )
   > SplineIK = RigVMFunctionReferenceNode(
    Bones=((Type=Bone,Name="hair_strand_side_01_l"),(Type=Bone,Name="hair_strand_side_02_l"),(Type=Bone,Name="hair_strand_side_03_l"),(Type=Bone,Name="hair_strand_side_04_l"),(Type=Bone,Name="hair_strand_side_05_l"))
    Controls=((Type=Control,Name="hair_strand_side_01_l_ctrl"),(Type=Control,Name="hair_strand_side_02_l_ctrl"),(Type=Control,Name="hair_strand_side_03_l_ctrl"),(Type=Control,Name="hair_strand_side_04_l_ctrl"),(Type=Control,Name="hair_strand_side_05_l_ctrl"))
    Primary Axis=(X=1,Y=0,Z=0)
    Samples Per Segment=8
    Secondary Spline Direction=(X=0,Y=5,Z=0)
    Spline Mode=BSpline
    Stretch=false
    Up Axis=(X=0,Y=1,Z=0)
    debug=false
   )
   SequenceAggregate_3 = RigVMAggregateNode(ExecuteContext=@SequenceAggregate_1.A)
   n9 = &RigUnit_SphericalPoseReader(
    ActiveRegionScaleFactors=(PositiveWidth=2,NegativeWidth=2,PositiveHeight=2,NegativeHeight=0.4)
    ActiveRegionSize=0.1
    Debug=(bDrawDebug=false,DebugScale=25,DebugSegments=20,DebugThickness=0.25)
    DriverAxis=(X=0,Y=1,Z=0)
    DriverItem=(Type=Bone,Name="head")
    FalloffRegionScaleFactors=(PositiveWidth=2,NegativeWidth=2,PositiveHeight=1,NegativeHeight=0.4)
    FalloffSize=0.2
    OptionalParentItem=(Type=Bone)
    RotationOffset=(X=110,Y=0,Z=7.5)
    ExecutePin=@SequenceAggregate_3.A
   )
   > RigVMVariableNode(Value=@n9.OutputParam) {variable=spr_look_left_alpha}
   n10 = &RigUnit_SphericalPoseReader(
    ActiveRegionScaleFactors=(PositiveWidth=1,NegativeWidth=1,PositiveHeight=1,NegativeHeight=1)
    ActiveRegionSize=0.1
    Debug=(bDrawDebug=false,DebugScale=25,DebugSegments=20,DebugThickness=0.25)
    DriverAxis=(X=1,Y=0,Z=0)
    DriverItem=(Type=Bone,Name="clavicle_l")
    FalloffRegionScaleFactors=(PositiveWidth=1,NegativeWidth=1,PositiveHeight=1,NegativeHeight=1)
    FalloffSize=0.2
    OptionalParentItem=(Type=Bone)
    RotationOffset=(X=46,Y=0,Z=-83)
    ExecutePin=@SequenceAggregate_3.D
   )
   > RigVMVariableNode(Value=@n10.OutputParam) {variable=spr_clavicle_l_up_alpha}
   n11 = &RigUnit_SphericalPoseReader(
    ActiveRegionScaleFactors=(PositiveWidth=3,NegativeWidth=0.4,PositiveHeight=3,NegativeHeight=3)
    ActiveRegionSize=0.1
    Debug=(bDrawDebug=false,DebugScale=25,DebugSegments=20,DebugThickness=0.25)
    DriverAxis=(X=0,Y=1,Z=0)
    DriverItem=(Type=Bone,Name="head")
    FalloffRegionScaleFactors=(PositiveWidth=1,NegativeWidth=0.23,PositiveHeight=0.5,NegativeHeight=0.5)
    FalloffSize=0.35
    OptionalParentItem=(Type=Bone)
    RotationOffset=(X=90,Y=0,Z=-22)
    ExecutePin=@SequenceAggregate_3.B
   )
   > RigVMVariableNode(Value=@n11.OutputParam) {variable=spr_look_up_alpha}
   n12 = &RigUnit_SphericalPoseReader(
    ActiveRegionScaleFactors=(PositiveWidth=1,NegativeWidth=1,PositiveHeight=1,NegativeHeight=1)
    ActiveRegionSize=0.025
    Debug=(bDrawDebug=false,DebugScale=25,DebugSegments=20,DebugThickness=0.25)
    DriverAxis=(X=1,Y=0,Z=0)
    DriverItem=(Type=Bone,Name="neck_01")
    FalloffRegionScaleFactors=(PositiveWidth=1,NegativeWidth=1,PositiveHeight=1,NegativeHeight=1)
    FalloffSize=0.05
    OptionalParentItem=(Type=Bone)
    RotationOffset=(X=9,Y=-90,Z=0)
    ExecutePin=@SequenceAggregate_3.E
   )
   > RigVMVariableNode(Value=@n12.OutputParam) {variable=spr_neck_down_alpha}
   n13 = &RigUnit_SphericalPoseReader(
    ActiveRegionScaleFactors=(PositiveWidth=5,NegativeWidth=1,PositiveHeight=3.5,NegativeHeight=5)
    ActiveRegionSize=0.1
    Debug=(bDrawDebug=false,DebugScale=25,DebugSegments=20,DebugThickness=0.25)
    DriverAxis=(X=-1,Y=0,Z=0)
    DriverItem=(Type=Bone,Name="upperarm_r")
    FalloffRegionScaleFactors=(PositiveWidth=1,NegativeWidth=0.75,PositiveHeight=1,NegativeHeight=1)
    FalloffSize=0.5
    FlipWidthScaling=false
    OptionalParentItem=(Type=Bone)
    RotationOffset=(X=-90,Y=25,Z=10)
    ExecutePin=@SequenceAggregate_3.C
   )
   > RigVMVariableNode(Value=@n13.OutputParam) {variable=spr_upperarm_r_up_pistol_alpha}
  ]}
 RigVMModel Rig Graph {nodes [
   n1 = &RigUnit_PrepareForExecution()
   SequenceAggregate_3 = &RigVMFunction_Sequence(ExecuteContext=@n1.ExecutePin)
   SequenceAggregate_1_1 = RigVMAggregateNode(ExecuteContext=@SequenceAggregate_3.A)
   &RigUnit_ModifyTransforms(
    ItemToModify=((Item=(Type=Control,Name="hair_strand_side_04_l_ctrl"),Transform=(Rotation=(X=0,Y=0,Z=0,W=1),Translation=(X=-1,Y=0,Z=-1.5),Scale3D=(X=1,Y=1,Z=1))))
    Weight=1
    ExecutePin=@SequenceAggregate_1_1.B
   )
   n2 = /Script/ControlRig.RigUnit_HierarchyGetPoseItemArray(
    ElementType=All
    Initial=false
    ItemsToGet=((Type=Control,Name="hair_strand_side_01_l_ctrl"),(Type=Control,Name="hair_strand_side_02_l_ctrl"),(Type=Control,Name="hair_strand_side_03_l_ctrl"),(Type=Control,Name="hair_strand_side_04_l_ctrl"),(Type=Control,Name="hair_strand_side_05_l_ctrl"))
   ) {is=struct RigVMUnitNode}
   > RigVMVariableNode(Value=@n2.Pose) {variable=left_strand_spline_ik_bind_pose}
   n3 = &RigUnit_HierarchyGetPoseItemArray(
    ElementType=All
    ItemsToGet=((Type=Control,Name="dyn_ponytail_01_ctrl"),(Type=Control,Name="dyn_ponytail_02_ctrl"),(Type=Control,Name="dyn_ponytail_03_ctrl"))
   )
   RigVMVariableNode(Value=@n3.Pose, ExecuteContext=@SequenceAggregate_1_1.D) {variable=ponytail_spline_ik_bind_pose}
   n4 = &RigUnit_HierarchyGetPoseItemArray(
    ElementType=All
    ItemsToGet="((Type=Control,Name=\"dyn_ponytail_tip_mstr_ctrl\"),(Type=Control,Name=\"dyn_ponytail_tip_01_ctrl\"),(Type=Control,Name=\"dyn_ponytail_tip_02_ctrl\"),(Type=Control,Name=\"dyn_ponytail_tip_03_ctrl\"),(Type=Control,Name=\"dyn_ponytail_tip_04_ctrl\"),(Type=Control,Name ...(truncated)"
   )
   RigVMVariableNode(Value=@n4.Pose, ExecuteContext=@SequenceAggregate_1_1.E) {variable=ponytail_tip_spline_ik_bind_pose}
   n5 = &RigUnit_HierarchyGetPoseItemArray(
    ElementType=All
    Initial=false
    ItemsToGet=((Type=Control,Name="hair_strand_side_01_r_ctrl"),(Type=Control,Name="hair_strand_side_02_r_ctrl"),(Type=Control,Name="hair_strand_side_03_r_ctrl"),(Type=Control,Name="hair_strand_side_04_r_ctrl"),(Type=Control,Name="hair_strand_side_05_r_ctrl"))
   )
   RigVMVariableNode(Value=@n5.Pose, ExecuteContext=@SequenceAggregate_1_1.C) {variable=right_strand_spline_ik_bind_pose}
   SequenceAggregate_2 = RigVMAggregateNode(ExecuteContext=@SequenceAggregate_3.B)
   &RigUnit_ModifyTransforms(
    ItemToModify=((Item=(Type=Control,Name="hair_strand_side_03_l_ctrl"),Transform=(Rotation=(X=0,Y=0,Z=0,W=1),Translation=(X=0,Y=0,Z=-1.5),Scale3D=(X=1,Y=1,Z=1))))
    Weight=1
    ExecutePin=@SequenceAggregate_2.A
   )
   > &RigUnit_ModifyTransforms(
    ItemToModify=((Item=(Type=Control,Name="hair_strand_side_04_l_ctrl"),Transform=(Rotation=(X=0,Y=0,Z=0,W=1),Translation=(X=0,Y=0,Z=-1),Scale3D=(X=1,Y=1,Z=1))))
    Weight=1
   )
   > &RigUnit_ModifyTransforms(
    ItemToModify=((Item=(Type=Control,Name="hair_strand_side_05_l_ctrl"),Transform=(Rotation=(X=0,Y=0,Z=0,W=1),Translation=(X=0,Y=0,Z=-1),Scale3D=(X=1,Y=1,Z=1))))
    Weight=1
   )
   n6 = &RigUnit_HierarchyGetPoseItemArray(
    ElementType=All
    Initial=false
    ItemsToGet=((Type=Control,Name="hair_strand_side_03_l_ctrl"),(Type=Control,Name="hair_strand_side_04_l_ctrl"),(Type=Control,Name="hair_strand_side_05_l_ctrl"))
   )
   > RigVMVariableNode(Value=@n6.Pose) {variable=left_strand_spline_ik_look_left_pose}
   SequenceAggregate_1 = RigVMAggregateNode(ExecuteContext=@SequenceAggregate_2.E)
   &RigUnit_ModifyTransforms(
    ItemToModify=((Item=(Type=Control,Name="hair_strand_side_03_l_ctrl"),Transform=(Rotation=(X=0,Y=0,Z=0,W=1),Translation=(X=-1,Y=2.5,Z=0),Scale3D=(X=1,Y=1,Z=1))))
    Weight=1
    ExecutePin=@SequenceAggregate_1.A
   )
   > &RigUnit_ModifyTransforms(
    ItemToModify=((Item=(Type=Control,Name="hair_strand_side_04_l_ctrl"),Transform=(Rotation=(X=0,Y=0,Z=0,W=1),Translation=(X=1,Y=0,Z=1.5),Scale3D=(X=1,Y=1,Z=1))))
    Weight=1
   )
   > &RigUnit_ModifyTransforms(
    ItemToModify=((Item=(Type=Control,Name="hair_strand_side_05_l_ctrl"),Transform=(Rotation=(X=0,Y=0,Z=0,W=1),Translation=(X=0,Y=3,Z=0),Scale3D=(X=1,Y=1,Z=1))))
    Weight=1
   )
   n7 = &RigUnit_HierarchyGetPoseItemArray(
    ElementType=All
    Initial=false
    ItemsToGet=((Type=Control,Name="hair_strand_side_03_l_ctrl"),(Type=Control,Name="hair_strand_side_04_l_ctrl"),(Type=Control,Name="hair_strand_side_05_l_ctrl"))
   )
   > RigVMVariableNode(Value=@n7.Pose) {variable=left_strand_FE_pose_adjustment}
   &RigUnit_ModifyTransforms(
    ItemToModify=((Item=(Type=Control,Name="hair_strand_side_03_r_ctrl"),Transform=(Rotation=(X=0,Y=0,Z=0,W=1),Translation=(X=0,Y=0,Z=0),Scale3D=(X=1,Y=1,Z=1))))
    Weight=1
    ExecutePin=@SequenceAggregate_1.B
   )
   > &RigUnit_ModifyTransforms(
    ItemToModify=((Item=(Type=Control,Name="hair_strand_side_04_r_ctrl"),Transform=(Rotation=(X=0,Y=0,Z=0,W=1),Translation=(X=0,Y=0,Z=0),Scale3D=(X=1,Y=1,Z=1))))
    Weight=1
   )
   > &RigUnit_ModifyTransforms(
    ItemToModify=((Item=(Type=Control,Name="hair_strand_side_05_r_ctrl"),Transform=(Rotation=(X=0,Y=0,Z=0,W=1),Translation=(X=0,Y=0,Z=0),Scale3D=(X=1,Y=1,Z=1))))
    Weight=1
   )
   n8 = &RigUnit_HierarchyGetPoseItemArray(
    ElementType=All
    Initial=false
    ItemsToGet=((Type=Control,Name="hair_strand_side_03_r_ctrl"),(Type=Control,Name="hair_strand_side_04_r_ctrl"),(Type=Control,Name="hair_strand_side_05_r_ctrl"))
   )
   > RigVMVariableNode(Value=@n8.Pose) {variable=right_strand_FE_pose_adjustment}
   &RigUnit_ModifyTransforms(
    ItemToModify=((Item=(Type=Control,Name="hair_strand_side_03_l_ctrl"),Transform=(Rotation=(X=0,Y=0,Z=0,W=1),Translation=(X=0,Y=-1,Z=-1),Scale3D=(X=1,Y=1,Z=1))))
    Weight=1
    ExecutePin=@SequenceAggregate_1.C
   )
   > &RigUnit_ModifyTransforms(
    ItemToModify=((Item=(Type=Control,Name="hair_strand_side_04_l_ctrl"),Transform=(Rotation=(X=0,Y=0,Z=0,W=1),Translation=(X=1,Y=0,Z=1.5),Scale3D=(X=1,Y=1,Z=1))))
    Weight=1
   )
   > &RigUnit_ModifyTransforms(
    ItemToModify=((Item=(Type=Control,Name="hair_strand_side_05_l_ctrl"),Transform=(Rotation=(X=0,Y=0,Z=0,W=1),Translation=(X=0,Y=-1.75,Z=-1.25),Scale3D=(X=1,Y=1,Z=1))))
    Weight=1
   )
   n9 = &RigUnit_HierarchyGetPoseItemArray(
    ElementType=All
    Initial=false
    ItemsToGet=((Type=Control,Name="hair_strand_side_03_l_ctrl"),(Type=Control,Name="hair_strand_side_04_l_ctrl"),(Type=Control,Name="hair_strand_side_05_l_ctrl"))
   )
   > RigVMVariableNode(Value=@n9.Pose) {variable=left_strand_FE_pose_emote_adjustment}
   &RigUnit_ModifyTransforms(
    ItemToModify=((Item=(Type=Control,Name="hair_strand_side_03_r_ctrl"),Transform=(Rotation=(X=0,Y=0,Z=0,W=1),Translation=(X=1,Y=-1.5,Z=1),Scale3D=(X=1,Y=1,Z=1))))
    Weight=1
    ExecutePin=@SequenceAggregate_1.D
   )
   > &RigUnit_ModifyTransforms(
    ItemToModify=((Item=(Type=Control,Name="hair_strand_side_04_r_ctrl"),Transform=(Rotation=(X=0,Y=0,Z=0,W=1),Translation=(X=0,Y=0,Z=0),Scale3D=(X=1,Y=1,Z=1))))
    Weight=1
   )
   > &RigUnit_ModifyTransforms(
    ItemToModify=((Item=(Type=Control,Name="hair_strand_side_05_r_ctrl"),Transform=(Rotation=(X=0,Y=0,Z=0,W=1),Translation=(X=1,Y=-1,Z=-1),Scale3D=(X=1,Y=1,Z=1))))
    Weight=1
   )
   n10 = &RigUnit_HierarchyGetPoseItemArray(
    ElementType=All
    Initial=false
    ItemsToGet=((Type=Control,Name="hair_strand_side_03_r_ctrl"),(Type=Control,Name="hair_strand_side_04_r_ctrl"),(Type=Control,Name="hair_strand_side_05_r_ctrl"))
   )
   > RigVMVariableNode(Value=@n10.Pose) {variable=right_strand_FE_pose_emote_adjustment}
   SequenceAggregate = RigVMAggregateNode(ExecuteContext=@SequenceAggregate_2.F)
   &RigUnit_ModifyTransforms(
    ItemToModify=((Item=(Type=Control,Name="dyn_ponytail_tip_mstr_ctrl"),Transform=(Rotation=(X=0,Y=0,Z=0,W=1),Translation=(X=0,Y=0,Z=2),Scale3D=(X=1,Y=1,Z=1))))
    Weight=1
    ExecutePin=@SequenceAggregate.A
   )
   > &RigUnit_ModifyTransforms(
    ItemToModify=((Item=(Type=Control,Name="dyn_ponytail_tip_02_ctrl"),Transform=(Rotation=(X=0,Y=0,Z=0,W=1),Translation=(X=0,Y=0,Z=1),Scale3D=(X=1,Y=1,Z=1))))
    Weight=1
   )
   > &RigUnit_ModifyTransforms(
    ItemToModify=((Item=(Type=Control,Name="dyn_ponytail_tip_03_ctrl"),Transform=(Rotation=(X=0,Y=0,Z=0,W=1),Translation=(X=0,Y=-1,Z=-11),Scale3D=(X=1,Y=1,Z=1))))
    Weight=1
   )
   n11 = &RigUnit_HierarchyGetPoseItemArray(
    ElementType=All
    ItemsToGet=((Type=Control,Name="dyn_ponytail_tip_mstr_ctrl"),(Type=Control,Name="dyn_ponytail_tip_02_ctrl"),(Type=Control,Name="dyn_ponytail_tip_03_ctrl"))
   )
   > RigVMVariableNode(Value=@n11.Pose) {variable=ponytail_tip_spline_ik_up_right_pstl}
   &RigUnit_ModifyTransforms(
    ItemToModify=((Item=(Type=Control,Name="hair_strand_side_03_r_ctrl"),Transform=(Rotation=(X=0,Y=0,Z=0,W=1),Translation=(X=0,Y=0.75,Z=-2),Scale3D=(X=1,Y=1,Z=1))))
    Weight=1
    ExecutePin=@SequenceAggregate.B
   )
   > &RigUnit_ModifyTransforms(
    ItemToModify=((Item=(Type=Control,Name="hair_strand_side_05_r_ctrl"),Transform=(Rotation=(X=0,Y=0,Z=0,W=1),Translation=(X=0,Y=-1,Z=-2.5),Scale3D=(X=1,Y=1,Z=1))))
    Weight=1
   )
   n12 = &RigUnit_HierarchyGetPoseItemArray(
    ElementType=All
    Initial=false
    ItemsToGet=((Type=Control,Name="hair_strand_side_03_r_ctrl"),(Type=Control,Name="hair_strand_side_05_r_ctrl"))
   )
   > RigVMVariableNode(Value=@n12.Pose) {variable=right_strand_spline_ik_up_right_pstl}
   &RigUnit_ModifyTransforms(
    ItemToModify=((Item=(Type=Control,Name="hair_strand_side_03_r_ctrl"),Transform=(Rotation=(X=0,Y=0,Z=0,W=1),Translation=(X=0,Y=0,Z=1),Scale3D=(X=1,Y=1,Z=1))))
    Weight=1
    ExecutePin=@SequenceAggregate.D
   )
   > &RigUnit_ModifyTransforms(
    ItemToModify=((Item=(Type=Control,Name="hair_strand_side_05_r_ctrl"),Transform=(Rotation=(X=0,Y=0,Z=0,W=1),Translation=(X=0,Y=0,Z=0.5),Scale3D=(X=1,Y=1,Z=1))))
    Weight=1
   )
   n13 = &RigUnit_HierarchyGetPoseItemArray(
    ElementType=All
    Initial=false
    ItemsToGet=((Type=Control,Name="hair_strand_side_03_r_ctrl"),(Type=Control,Name="hair_strand_side_05_r_ctrl"))
   )
   > RigVMVariableNode(Value=@n13.Pose) {variable=right_strand_spline_ik_neck_01_down}
   &RigUnit_ModifyTransforms(
    ItemToModify=((Item=(Type=Control,Name="hair_strand_side_03_l_ctrl"),Transform=(Rotation=(X=0,Y=0,Z=0,W=1),Translation=(X=0,Y=0,Z=0),Scale3D=(X=1,Y=1,Z=1))))
    Weight=1
    ExecutePin=@SequenceAggregate.C
   )
   > &RigUnit_ModifyTransforms(
    ItemToModify=((Item=(Type=Control,Name="hair_strand_side_05_l_ctrl"),Transform=(Rotation=(X=0,Y=0,Z=0,W=1),Translation=(X=0,Y=0,Z=0),Scale3D=(X=1,Y=1,Z=1))))
    Weight=1
   )
   n14 = &RigUnit_HierarchyGetPoseItemArray(
    ElementType=All
    Initial=false
    ItemsToGet=((Type=Control,Name="hair_strand_side_03_l_ctrl"),(Type=Control,Name="hair_strand_side_05_l_ctrl"))
   )
   > RigVMVariableNode(Value=@n14.Pose) {variable=left_strand_spline_ik_neck_01_down}
   &RigUnit_ModifyTransforms(
    ItemToModify=((Item=(Type=Control,Name="hair_strand_side_03_l_ctrl"),Transform=(Rotation=(X=0,Y=0,Z=0,W=1),Translation=(X=2,Y=-1,Z=1),Scale3D=(X=1,Y=1,Z=1))))
    Weight=1
    ExecutePin=@SequenceAggregate_2.C
   )
   > &RigUnit_ModifyTransforms(
    ItemToModify=((Item=(Type=Control,Name="hair_strand_side_05_l_ctrl"),Transform=(Rotation=(X=0,Y=0,Z=0,W=1),Translation=(X=0,Y=-1,Z=0),Scale3D=(X=1,Y=1,Z=1))))
    Weight=1
   )
   n15 = &RigUnit_HierarchyGetPoseItemArray(
    ElementType=All
    Initial=false
    ItemsToGet=((Type=Control,Name="hair_strand_side_03_l_ctrl"),(Type=Control,Name="hair_strand_side_05_l_ctrl"))
   )
   > RigVMVariableNode(Value=@n15.Pose) {variable=left_strand_spline_ik_clav_l_up}
   &RigUnit_ModifyTransforms(
    ItemToModify=((Item=(Type=Control,Name="hair_strand_side_03_l_ctrl"),Transform=(Rotation=(X=0,Y=0,Z=0,W=1),Translation=(X=0,Y=-1.5,Z=0.75),Scale3D=(X=1,Y=1,Z=1))))
    Weight=1
    ExecutePin=@SequenceAggregate_2.D
   )
   > &RigUnit_ModifyTransforms(
    ItemToModify=((Item=(Type=Control,Name="hair_strand_side_04_l_ctrl"),Transform=(Rotation=(X=0,Y=0,Z=0,W=1),Translation=(X=0,Y=0,Z=0),Scale3D=(X=1,Y=1,Z=1))))
    Weight=1
   )
   > &RigUnit_ModifyTransforms(
    ItemToModify=((Item=(Type=Control,Name="hair_strand_side_05_l_ctrl"),Transform=(Rotation=(X=0,Y=0,Z=0,W=1),Translation=(X=0,Y=-1.2,Z=0.5),Scale3D=(X=1,Y=1,Z=1))))
    Weight=1
   )
   n16 = &RigUnit_HierarchyGetPoseItemArray(
    ElementType=All
    Initial=false
    ItemsToGet=((Type=Control,Name="hair_strand_side_03_l_ctrl"),(Type=Control,Name="hair_strand_side_04_l_ctrl"),(Type=Control,Name="hair_strand_side_05_l_ctrl"))
   )
   > RigVMVariableNode(Value=@n16.Pose) {variable=left_strand_spline_ik_clav_up_head_up}
   &RigUnit_ModifyTransforms(
    ItemToModify=((Item=(Type=Control,Name="hair_strand_side_01_l_ctrl"),Transform=(Rotation=(X=0,Y=0,Z=0,W=1),Translation=(X=0,Y=0,Z=0),Scale3D=(X=1,Y=1,Z=1))))
    Weight=1
    ItemToModify.0.Transform.Rotation=&RigVMFunction_MathQuaternionFromEuler(Euler=(X=0,Y=0,Z=45)).Result
    ExecutePin=@SequenceAggregate_2.B
   )
   > RigVMVariableNode(Value=&RigUnit_HierarchyGetPoseItemArray(ElementType=All, Initial=false, ItemsToGet=((Type=Control,Name="hair_strand_side_01_l_ctrl"))).Pose) {variable=left_strand_spline_ik_look_up}
  ]}
]
hierarchy [
 root {loc=<0 0 0>, rot=<0 0 0 1>}
 pelvis {parent=root, loc=<0 -0.5301 98.0571>, rot=<0.0127 0.707 -0.0127 -0.707>}
 spine_01 {parent=pelvis, loc=<6.1208 -0.9328 0>, rot=<0 0 0.034 -0.9994>}
 spine_02 {parent=spine_01, loc=<6.5996 -0.2095 0>, rot=<0 0 0.0238 -0.9997>}
 spine_03 {parent=spine_02, loc=<6.4914 -0.1237 0>, rot=<0 0 0.0149 -0.9999>}
 spine_04 {parent=spine_03, loc=<6.9598 0.0794 0>, rot=<0 0 -0.0132 -0.9999>}
 spine_05 {parent=spine_04, loc=<6.9987 0 0>, rot=<0 0 -0.0015 -1>}
 clavicle_l {parent=spine_05, loc=<5.4831 -0.1053 -3.2946>, rot=<0.0465 -0.7289 -0.009 -0.6829>}
 upperarm_l {parent=clavicle_l, loc=<9.8658 -1.2136 -0.4105>, rot=<0.0164 -0.3245 -0.0269 -0.9454>}
 neck_01 {parent=spine_05, loc=<11.3138 0.1529 0>, rot=<0 0 -0.2003 -0.9797>}
 neck_02 {parent=neck_01, loc=<4.5627 0 0>, rot=<0 0 0.0504 -0.9987>}
 head {parent=neck_02, loc=<5.3914 0 0>, rot=<0 0 0.115 -0.9934>}
 hair_strand_side_mstr_l {parent=head, loc=<3.9063 -2.6761 -6.557>, rot=<-0.0043 0.0354 0.9971 -0.0678>}
 hair_strand_side_01_l {parent=hair_strand_side_mstr_l, loc=<0 0 0>, rot=<0 0 0 1>}
 hair_strand_side_02_l {parent=hair_strand_side_01_l, loc=<5.5 0 0>, rot=<-0.1834 -0.0102 0.2202 -0.958>}
 hair_strand_side_03_l {parent=hair_strand_side_02_l, loc=<5.5 0 0>, rot=<0.1134 0.1464 0.1217 -0.9751>}
 hair_strand_side_04_l {parent=hair_strand_side_03_l, loc=<5.5 0 0>, rot=<0.0217 -0.0177 0.0206 -0.9994>}
 hair_strand_side_05_l {parent=hair_strand_side_04_l, loc=<5.5 0 0>, rot=<0.012 -0.0597 -0.025 -0.9978>}
 hair_strand_side_mstr_r {parent=head, loc=<3.8558 -2.3079 6.5305>, rot=<0.0042 0.0397 0.9974 -0.0601>}
 hair_strand_side_01_r {parent=hair_strand_side_mstr_r, loc=<0 0 0>, rot=<0 0 0 1>}
 hair_strand_side_02_r {parent=hair_strand_side_01_r, loc=<5.4996 0.003 0.0424>, rot=<-0.1888 -0.0369 0.2522 -0.9484>}
 hair_strand_side_03_r {parent=hair_strand_side_02_r, loc=<5.3041 0.5924 1.328>, rot=<0.112 0.1527 0.0796 -0.9787>}
 hair_strand_side_04_r {parent=hair_strand_side_03_r, loc=<5.4344 -0.1508 -0.8334>, rot=<0.0143 -0.0103 -0.0324 -0.9993>}
 hair_strand_side_05_r {parent=hair_strand_side_04_r, loc=<5.4862 -0.0538 -0.3898>, rot=<0.028 0.0294 -0.0341 -0.9986>}
 hair_strand_back_mstr_l {parent=head, loc=<11.9857 -7.6352 -3.094>, rot=<0.5474 -0.7701 -0.3001 -0.1313>}
 hair_strand_back_01_l {parent=hair_strand_back_mstr_l, loc=<0 0 0>, rot=<0 0 0 1>}
 hair_strand_back_02_l {parent=hair_strand_back_01_l, loc=<5.5 0 0>, rot=<0.0503 -0.0163 -0.4349 -0.8989>}
 hair_strand_back_03_l {parent=hair_strand_back_02_l, loc=<5.5 0 0>, rot=<-0.1496 0.0805 -0.2031 -0.9643>}
 hair_strand_back_04_l {parent=hair_strand_back_03_l, loc=<5.5 0 0>, rot=<0.1013 0.0492 -0.0132 -0.9935>}
 hair_strand_back_05_l {parent=hair_strand_back_04_l, loc=<5.5 0 0>, rot=<-0.1688 0.034 0.0527 -0.9837>}
 hair_strand_back_06_l {parent=hair_strand_back_05_l, loc=<5.5 0 0>, rot=<0.1504 -0.0508 0.0687 -0.9849>}
 hair_strand_back_mstr_r {parent=head, loc=<11.9853 -7.6352 3.094>, rot=<0.5474 -0.7701 0.3001 0.1313>}
 hair_strand_back_01_r {parent=hair_strand_back_mstr_r, loc=<0 0 0>, rot=<0 0 0 1>}
 hair_strand_back_02_r {parent=hair_strand_back_01_r, loc=<5.5001 0 0>, rot=<-0.0503 0.0163 -0.4349 -0.8989>}
 hair_strand_back_03_r {parent=hair_strand_back_02_r, loc=<5.5 0 0>, rot=<0.1496 -0.0805 -0.2031 -0.9643>}
 hair_strand_back_04_r {parent=hair_strand_back_03_r, loc=<5.4998 0 0>, rot=<-0.1013 -0.0492 -0.0132 -0.9935>}
 hair_strand_back_05_r {parent=hair_strand_back_04_r, loc=<5.4998 0 0>, rot=<0.1688 -0.0339 0.0527 -0.9837>}
 hair_strand_back_06_r {parent=hair_strand_back_05_r, loc=<5.5002 0 0>, rot=<-0.1504 0.0508 0.0687 -0.9849>}
 dyn_ponytail_mstr {parent=head, loc=<11.4435 -7.9638 0.0018>, rot=<0.7355 -0.6775 0 0>}
 dyn_ponytail_base {parent=dyn_ponytail_mstr, loc=<0 0 0>, rot=<0 0 0 1>}
 dyn_ponytail_01 {parent=dyn_ponytail_base, loc=<4.8313 0 0>, rot=<0.0257 0.0998 -0.6235 -0.775>}
 dyn_ponytail_02 {parent=dyn_ponytail_01, loc=<6.9684 0 0>, rot=<0.0352 -0.0211 -0.2544 -0.9662>}
 dyn_ponytail_03 {parent=dyn_ponytail_02, loc=<7.0615 0 0>, rot=<0.1352 -0.1748 -0.163 -0.9616>}
 dyn_ponytail_04 {parent=dyn_ponytail_03, loc=<6.883 0 0>, rot=<-0.4206 -0.2473 -0.2837 -0.8255>}
 dyn_ponytail_05 {parent=dyn_ponytail_04, loc=<6.9592 0 0>, rot=<-0.7252 -0.0931 -0.2063 -0.6503>}
 dyn_ponytail_tip_mstr {parent=dyn_ponytail_05, loc=<6.3146 0 0>, rot=<-0.0076 0.0075 -0.2211 -0.9752>}
 dyn_ponytail_tip_01 {parent=dyn_ponytail_tip_mstr, loc=<0 0 0>, rot=<0 0 -0.0203 -0.9998>}
 dyn_ponytail_tip_02 {parent=dyn_ponytail_tip_01, loc=<3.8706 0 0>, rot=<-0.0004 0.0068 -0.1448 -0.9894>}
 dyn_ponytail_tip_03 {parent=dyn_ponytail_tip_02, loc=<3.8496 0 0>, rot=<-0.0673 0.0577 -0.1465 -0.9852>}
 dyn_ponytail_tip_04 {parent=dyn_ponytail_tip_03, loc=<3.8574 0 0>, rot=<-0.0394 0.1195 -0.0453 -0.991>}
 dyn_ponytail_tip_05 {parent=dyn_ponytail_tip_04, loc=<3.8904 0 0>, rot=<0 0 0 1>}
 dyn_ponytail_tip_scale_05 {parent=dyn_ponytail_tip_05, loc=<0 0 0>, rot=<0 0 0 1>, scale=<1 1 0.6338>}
 dyn_ponytail_tip_scale_04 {parent=dyn_ponytail_tip_04, loc=<0 0 0>, rot=<0 0 0 1>, scale=<1 1 0.6338>}
 dyn_ponytail_tip_scale_03 {parent=dyn_ponytail_tip_03, loc=<0 0 0>, rot=<0 0 0 1>}
 dyn_ponytail_tip_scale_02 {parent=dyn_ponytail_tip_02, loc=<0 0 0>, rot=<0 0 0 1>, scale=<1 1 0.6338>}
 dyn_ponytail_tip_scale_01 {parent=dyn_ponytail_tip_01, loc=<0 0 0>, rot=<0 0 0 1>}
 dyn_hair_strand_01 {parent=dyn_ponytail_base, loc=<-0.7829 -3.7632 0.8109>, rot=<-0.1816 0.3665 0.5427 -0.7336>}
 dyn_hair_strand_02 {parent=dyn_ponytail_base, loc=<-1.0864 -3.5588 0.4231>, rot=<0.3028 -0.9159 0.0683 -0.2544>}
 dyn_hair_strand_03 {parent=dyn_ponytail_base, loc=<4.0385 -3.356 4.5535>, rot=<0.1375 0.3456 -0.244 -0.8956>}
 dyn_hair_strand_04 {parent=dyn_ponytail_base, loc=<3.4796 -6.1852 -0.3526>, rot=<0.0029 0.0129 0.0816 -0.9966>}
 dyn_hair_strand_05 {parent=dyn_ponytail_base, loc=<5.1132 -4.5986 -3.6762>, rot=<-0.2743 0.0144 -0.136 -0.9519>}
 dyn_hair_strand_06 {parent=dyn_ponytail_base, loc=<0.0528 -1.9865 -3.1112>, rot=<-0.3819 -0.2822 0.5314 -0.7015>}
 hair_strand_side_01_l_ctrl {controlType=EULER_TRANSFORM, parent=head, type=Control, loc=<6.557 -4.3956 155.9183>, rot=<0.7082 -0.07 -0.7023 0.02>}
 hair_strand_side_01_r_ctrl {controlType=EULER_TRANSFORM, parent=head, type=Control, loc=<-6.5305 -4.027 155.8708>, rot=<0.7023 -0.0677 -0.7085 0.0115>}
 dyn_ponytail_01_ctrl {controlType=EULER_TRANSFORM, parent=head, type=Control, loc=<-0.0018 -14.5636 163.7675>, rot=<-0.1655 0.6613 -0.0374 0.7307>}
 dyn_earring_l {parent=head, loc=<0.3535 1.0888 -6.0324>, rot=<0.0079 -0.7071 -0.7071 0.0089>}
 dyn_earring_r {parent=head, loc=<0.3538 1.0888 6.0324>, rot=<0.0079 -0.7071 -0.7071 0.0089>}
 dfrm_collar_l {parent=neck_01, loc=<4.5999 1.9086 -5.6788>, rot=<0 0 -0.0049 -1>}
 hair_strand_side_03_l_ctrl {controlType=EULER_TRANSFORM, parent=dfrm_collar_l, type=Control, loc=<7.3119 -3.4907 145.2539>, rot=<-0.6259 -0.1603 0.7162 0.264>}
 dfrm_collar_r {parent=neck_01, loc=<4.5999 1.9086 5.679>, rot=<0 0 -0.0049 -1>}
 hair_strand_side_03_r_ctrl {controlType=EULER_TRANSFORM, parent=dfrm_collar_r, type=Control, loc=<-7.0662 -2.6729 145.2875>, rot=<-0.6342 -0.1535 0.7098 0.2655>}
 clavicle_r {parent=spine_05, loc=<5.4831 -0.1053 3.2946>, rot=<0.7289 0.0465 -0.6829 0.009>}
 upperarm_r {parent=clavicle_r, loc=<-9.8658 1.2136 0.4105>, rot=<0.0164 -0.3245 -0.0269 -0.9454>}
 dyn_ponytail_tip_mstr_ctrl_null {parent=clavicle_r, type=Null, loc=<-6.4309 -1.9412 -5.8109>, rot=<-0.1589 0.6042 0.4879 -0.6096>}
 dyn_ponytail_tip_mstr_ctrl {controlType=EULER_TRANSFORM, parent=dyn_ponytail_tip_mstr_ctrl_null, type=Control, loc=<0 0 0>, rot=<0 0 0 1>}
 dyn_ponytail_tip_01_ctrl {
  controlType=EULER_TRANSFORM
  parent=dyn_ponytail_tip_mstr_ctrl
  type=Control
  loc=<-10.19 -2.5179 141.9824>
  rot=<-0.6271 -0.4961 0.5893 0.1158>
 }
 dyn_ponytail_05_ctrl_null {parent=dyn_ponytail_tip_mstr_ctrl, type=Null, loc=<-5.6964 2.724 0.0714>, rot=<-0.0076 0.0075 -0.2211 0.9752>}
 dyn_ponytail_05_ctrl {
  controlType=EULER_TRANSFORM
  parent=dyn_ponytail_05_ctrl_null
  type=Control
  loc=<-3.2946 -4.1138 136.616>
  rot=<0.9991 0.0252 0.0336 -0.0008>
 }
 dyn_ponytail_02_ctrl_null {parent=spine_05, type=Null, loc=<26.9111 -10.779 -0.8523>, rot=<0.0685 0.9889 0.1315 -0.0067>}
 dyn_ponytail_02_ctrl {controlType=EULER_TRANSFORM, parent=dyn_ponytail_02_ctrl_null, type=Control, loc=<0 0 0>, rot=<0 0 0 1>}
 dyn_ponytail_03_ctrl_null {parent=spine_05, type=Null, loc=<19.9164 -9.8349 -0.6311>, rot=<0.205 0.9208 0.271 -0.1915>}
 dyn_ponytail_03_ctrl {controlType=EULER_TRANSFORM, parent=dyn_ponytail_03_ctrl_null, type=Control, loc=<0 0 0>, rot=<0 0 0 1>}
 dyn_ponytail_04_ctrl_null {parent=spine_05, type=Null, loc=<14.1168 -7.9507 2.5612>, rot=<0.2829 0.7686 -0.1672 -0.5489>}
 dyn_ponytail_04_ctrl {controlType=EULER_TRANSFORM, parent=dyn_ponytail_04_ctrl_null, type=Control, loc=<0 0 0>, rot=<0 0 0 1>}
 hair_strand_side_05_l_ctrl_null {parent=spine_05, type=Null, loc=<4.7227 7.009 -6.6903>, rot=<-0.0225 0.0651 0.9649 0.2536>}
 hair_strand_side_05_l_ctrl {
  controlType=EULER_TRANSFORM
  parent=hair_strand_side_05_l_ctrl_null
  type=Control
  loc=<-83.3018 -35.8791 227.8904>
  rot=<-0.2466 0.2693 0.3041 -0.8799>
 }
 hair_strand_side_05_r_ctrl_null {parent=spine_05, type=Null, loc=<4.3606 7.2305 6.3723>, rot=<0.0637 0.0333 0.9779 0.1962>}
 hair_strand_side_05_r_ctrl {
  controlType=EULER_TRANSFORM
  parent=hair_strand_side_05_r_ctrl_null
  type=Control
  loc=<-102.1107 31.5332 38.6785>
  rot=<-0.2118 -0.3397 -0.2752 -0.8741>
 }
 dyn_ponytail_tip_02_ctrl_null {parent=spine_05, type=Null, loc=<8.0607 4.6486 10.9129>, rot=<-0.0486 -0.267 0.9253 0.2648>}
 dyn_ponytail_tip_02_ctrl {controlType=EULER_TRANSFORM, parent=dyn_ponytail_tip_02_ctrl_null, type=Control, loc=<0 0 0>, rot=<0 0 0 1>}
 dyn_ponytail_tip_03_ctrl_null {parent=spine_05, type=Null, loc=<4.769 6.6349 11.1106>, rot=<-0.0158 -0.2089 0.9712 0.1132>}
 dyn_ponytail_tip_03_ctrl {controlType=EULER_TRANSFORM, parent=dyn_ponytail_tip_03_ctrl_null, type=Control, loc=<0 0 0>, rot=<0 0 0 1>}
 dyn_ponytail_tip_04_ctrl_null {parent=spine_05, type=Null, loc=<1.0124 7.5084 11.1747>, rot=<0.0953 -0.1817 0.9777 0.0438>}
 dyn_ponytail_tip_04_ctrl {controlType=EULER_TRANSFORM, parent=dyn_ponytail_tip_04_ctrl_null, type=Control, loc=<0 0 0>, rot=<0 0 0 1>}
 dyn_ponytail_tip_05_ctrl_null {parent=spine_05, type=Null, loc=<-2.7923 7.7069 11.9619>, rot=<0.0953 -0.1817 0.9777 0.0438>}
 dyn_ponytail_tip_05_ctrl {controlType=EULER_TRANSFORM, parent=dyn_ponytail_tip_05_ctrl_null, type=Control, loc=<0 0 0>, rot=<0 0 0 1>}
 hair_strand_side_04_l_ctrl_null {parent=root, type=Null, loc=<6.8872 -0.3075 140.7888>, rot=<-0.6406 -0.184 0.6957 0.2678>}
 hair_strand_side_04_l_ctrl {controlType=EULER_TRANSFORM, parent=hair_strand_side_04_l_ctrl_null, type=Control, loc=<0 0 0>, rot=<0 0 0 1>}
 hair_strand_side_02_l_ctrl_null {parent=root, type=Null, loc=<6.5783 -5.0957 150.4631>, rot=<-0.7047 0.0397 0.6571 0.2647>}
 hair_strand_side_02_l_ctrl {controlType=EULER_TRANSFORM, parent=hair_strand_side_02_l_ctrl_null, type=Control, loc=<0 0 0>, rot=<0 0 0 1>}
 hair_strand_side_02_r_ctrl_null {parent=root, type=Null, loc=<-6.6455 -4.6395 150.4064>, rot=<-0.7115 0.0204 0.6361 0.2978>}
 hair_strand_side_02_r_ctrl {controlType=EULER_TRANSFORM, parent=hair_strand_side_02_r_ctrl_null, type=Control, loc=<0 0 0>, rot=<0 0 0 1>}
 hair_strand_side_04_r_ctrl_null {parent=root, type=Null, loc=<-6.5176 0.4566 140.7981>, rot=<-0.6498 -0.1402 0.7092 0.2349>}
 hair_strand_side_04_r_ctrl {controlType=EULER_TRANSFORM, parent=hair_strand_side_04_r_ctrl_null, type=Control, loc=<0 0 0>, rot=<0 0 0 1>}
]
variables [
 is_front_end {type=bool, default=False}
 is_playing_emote {type=bool, default=False}
 left_strand_FE_pose_adjustment {type=FRigPose, default=(Elements=,HierarchyTopologyVersion=-1,PoseHash=-1)}
 left_strand_FE_pose_emote_adjustment {type=FRigPose, default=(Elements=,HierarchyTopologyVersion=-1,PoseHash=-1)}
 left_strand_spline_ik_bind_pose {type=FRigPose, default=(Elements=,HierarchyTopologyVersion=-1,PoseHash=-1)}
 left_strand_spline_ik_clav_l_up {type=FRigPose, default=(Elements=,HierarchyTopologyVersion=-1,PoseHash=-1)}
 left_strand_spline_ik_clav_up_head_up {type=FRigPose, default=(Elements=,HierarchyTopologyVersion=-1,PoseHash=-1)}
 left_strand_spline_ik_look_left_pose {type=FRigPose, default=(Elements=,HierarchyTopologyVersion=-1,PoseHash=-1)}
 left_strand_spline_ik_look_up {type=FRigPose, default=(Elements=,HierarchyTopologyVersion=-1,PoseHash=-1)}
 left_strand_spline_ik_neck_01_down {type=FRigPose, default=(Elements=,HierarchyTopologyVersion=-1,PoseHash=-1)}
 ponytail_spline_ik_bind_pose {type=FRigPose, default=(Elements=,HierarchyTopologyVersion=-1,PoseHash=-1)}
 ponytail_spline_ik_up_right_arm {type=FRigPose, default=(Elements=,HierarchyTopologyVersion=-1,PoseHash=-1)}
 ponytail_tip_spline_ik_bind_pose {type=FRigPose, default=(Elements=,HierarchyTopologyVersion=-1,PoseHash=-1)}
 ponytail_tip_spline_ik_up_right_pstl {type=FRigPose, default=(Elements=,HierarchyTopologyVersion=-1,PoseHash=-1)}
 right_strand_FE_pose_adjustment {type=FRigPose, default=(Elements=,HierarchyTopologyVersion=-1,PoseHash=-1)}
 right_strand_FE_pose_emote_adjustment {type=FRigPose, default=(Elements=,HierarchyTopologyVersion=-1,PoseHash=-1)}
 right_strand_spline_ik_bind_pose {type=FRigPose, default=(Elements=,HierarchyTopologyVersion=-1,PoseHash=-1)}
 right_strand_spline_ik_neck_01_down {type=FRigPose, default=(Elements=,HierarchyTopologyVersion=-1,PoseHash=-1)}
 right_strand_spline_ik_up_right_pstl {type=FRigPose, default=(Elements=,HierarchyTopologyVersion=-1,PoseHash=-1)}
 spr_clavicle_l_up_alpha {type=double, default=0}
 spr_look_left_alpha {type=double, default=0}
 spr_look_up_alpha {type=double, default=0}
 spr_neck_down_alpha {type=double, default=0}
 spr_upperarm_r_up_pistol_alpha {type=double, default=0}
]
