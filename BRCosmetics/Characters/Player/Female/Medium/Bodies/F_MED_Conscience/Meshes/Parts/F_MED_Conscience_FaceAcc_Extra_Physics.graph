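# Physics asset for the F_MED_Conscience female cosmetic face accessory mesh, defining secondary motion for earrings and ponytail hair strands.
# Capsule bodies cover two earrings, six strands, the ponytail base, head and neck.
# Limited swing and twist constraints tie earrings to the head and strands 01–05 to the ponytail base.
# Strand 01 has a second, reversed constraint, while strand 06 and the ponytail-to-head link have none.
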PhysicsAsset /BRCosmetics/Characters/Player/Female/Medium/Bodies/F_MED_Conscience/Meshes/Parts/F_MED_Conscience_FaceAcc_Extra_Physics
bodies {
 dyn_earring_l [{shapeType=Capsule, length=3.582, radius=1.1324, shapeName=dyn_earring_l_capsule, center=<5.1635 -0.9644 0>}]
 dyn_earring_r [{shapeType=Capsule, length=3.582, radius=1.1324, shapeName=dyn_earring_l_capsule, center=<5.2203 0.9485 0>}]
 dyn_hair_strand_01 [{
   shapeType=Capsule
   length=6.7616
   radius=0.7637
   shapeName=dyn_hair_strand_01_capsule
   center=<3.715 0.0001 0>
   rotation=<rot 90 180 179.999>
  }]
 dyn_hair_strand_02 [{
   shapeType=Capsule
   length=6.7616
   radius=0.7637
   shapeName=dyn_hair_strand_01_capsule
   center=<3.715 0.0001 0>
   rotation=<rot 90 180 179.999>
  }]
 dyn_hair_strand_03 [{
   shapeType=Capsule
   length=6.7616
   radius=0.7637
   shapeName=dyn_hair_strand_01_capsule
   center=<3.715 0.0001 0>
   rotation=<rot 90 180 179.999>
  }]
 dyn_hair_strand_04 [{
   shapeType=Capsule
   length=6.7616
   radius=0.7637
   shapeName=dyn_hair_strand_01_capsule
   center=<3.715 0.0001 0>
   rotation=<rot 90 180 179.999>
  }]
 dyn_hair_strand_05 [{
   shapeType=Capsule
   length=6.7616
   radius=0.7637
   shapeName=dyn_hair_strand_01_capsule
   center=<3.715 0.0001 0>
   rotation=<rot 90 180 179.999>
  }]
 dyn_hair_strand_06 [{
   shapeType=Capsule
   length=6.7616
   radius=0.7637
   shapeName=dyn_hair_strand_01_capsule
   center=<3.715 0.0001 0>
   rotation=<rot 90 180 179.999>
  }]
 dyn_ponytail_base [{shapeType=Capsule, length=0.3155, radius=0.8891, shapeName=dyn_ponytail_base_capsule}]
 head [{shapeType=Capsule, length=1.769, radius=2.6786, shapeName=head_capsule}]
 neck_01 [
  {shapeType=Capsule, length=6.4908, radius=5.9503, shapeName=neck_01_capsule, center=<4.9309 4.5833 0>, rotation=<rot 90 166.7373 166.7377>}
  {shapeType=Capsule, length=6.4908, radius=5.9503, shapeName=neck_01_capsule, center=<4.9308 0.4221 0>, rotation=<rot 90 90 90.0004>}
 ]
}
constraints [
 {bone1Name=dyn_hair_strand_01, bone2Name=dyn_ponytail_base, swing1LimitDegrees=10, swing2LimitDegrees=10, twistLimitDegrees=2}
 {bone1Name=dyn_ponytail_base, bone2Name=dyn_hair_strand_01, swing1LimitDegrees=10, swing2LimitDegrees=10, twistLimitDegrees=2}
 {bone1Name=dyn_hair_strand_03, bone2Name=dyn_ponytail_base, swing1LimitDegrees=10, swing2LimitDegrees=10, twistLimitDegrees=2}
 {bone1Name=dyn_hair_strand_04, bone2Name=dyn_ponytail_base, swing1LimitDegrees=10, swing2LimitDegrees=10, twistLimitDegrees=2}
 {bone1Name=dyn_hair_strand_05, bone2Name=dyn_ponytail_base, swing1LimitDegrees=10, swing2LimitDegrees=10, twistLimitDegrees=2}
 {bone1Name=dyn_hair_strand_02, bone2Name=dyn_ponytail_base, swing1LimitDegrees=10, swing2LimitDegrees=10, twistLimitDegrees=2}
 {bone1Name=dyn_earring_l, bone2Name=head, swing1LimitDegrees=15, swing2LimitDegrees=15, twistLimitDegrees=2}
 {bone1Name=dyn_earring_r, bone2Name=head, swing1LimitDegrees=15, swing2LimitDegrees=15, twistLimitDegrees=2}
]
