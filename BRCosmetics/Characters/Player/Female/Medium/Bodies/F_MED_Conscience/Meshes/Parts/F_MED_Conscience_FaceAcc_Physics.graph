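# Physics asset for the FaceAcc part of the female medium Conscience player body, covering its ponytail and two back hair strands.
# Capsules on head, upper spine, clavicles, ponytail bones and strand bones form the bodies.
# Each strand is a six-link constraint chain from a small master body, with tight twist throughout and narrower swing at the root.
# Ponytail, head, spine and clavicle bodies have no constraints here.
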PhysicsAsset /BRCosmetics/Characters/Player/Female/Medium/Bodies/F_MED_Conscience/Meshes/Parts/F_MED_Conscience_FaceAcc_Physics
bodies {
 clavicle_l [{
   shapeType=Capsule
   length=7.6238
   radius=5.4177
   shapeName=clavicle_l_capsule
   center=<4.961 -1.4598 0.1575>
   rotation=<rot -75 179.9984 -179.9988>
  }]
 clavicle_r [{
   shapeType=Capsule
   length=7.6238
   radius=5.4177
   shapeName=clavicle_l_capsule
   center=<-4.961 1.4598 -0.1575>
   rotation=<rot -75 179.9984 0.0012>
  }]
 dyn_ponytail_01 [{
   shapeType=Capsule
   length=3.785
   radius=4.5415
   shapeName=dyn_ponytail_01_capsule
   center=<-0.562 0.3124 0>
   rotation=<rot 90 -9.4623 45.5339>
  }]
 dyn_ponytail_02 [{shapeType=Capsule, length=3.9077, radius=4.2442, shapeName=dyn_ponytail_02_capsule, center=<0 0.622 0>, rotation=<rot 90 0 -0.0006>}]
 dyn_ponytail_03 [{
   shapeType=Capsule
   length=5.6734
   radius=3.7299
   shapeName=dyn_ponytail_03_capsule
   center=<2.6431 0.0001 0>
   rotation=<rot -90 180 -179.9985>
  }]
 dyn_ponytail_04 [{
   shapeType=Capsule
   length=4.8387
   radius=2.8498
   shapeName=dyn_ponytail_04_capsule
   center=<2.7473 0 0.6352>
   rotation=<rot -90 -5.051 5.0516>
  }]
 dyn_ponytail_base [{shapeType=Capsule, length=0.3155, radius=0.8891, shapeName=dyn_ponytail_base_capsule}]
 hair_strand_back_01_l [{
   shapeType=Capsule
   length=6.247
   radius=1.1049
   shapeName=hair_strand_back_01_l_capsule
   center=<3.091 0 0>
   rotation=<rot -90 179.9926 -179.9925>
  }]
 hair_strand_back_01_r [{
   shapeType=Capsule
   length=6.247
   radius=1.1049
   shapeName=hair_strand_back_01_l_capsule
   center=<3.091 0 0>
   rotation=<rot -90 179.9926 -179.9925>
  }]
 hair_strand_back_02_l [{shapeType=Capsule, length=6.247, radius=1.1049, shapeName=hair_strand_back_02_l_capsule, center=<3.091 0 0>, rotation=<rot -90 -90 90.0003>}]
 hair_strand_back_02_r [{
   shapeType=Capsule
   length=5.7371
   radius=1.1049
   shapeName=hair_strand_back_01_l_capsule
   center=<3.2362 0 0.1552>
   rotation=<rot -90 179.9996 -179.9995>
  }]
 hair_strand_back_03_l [{shapeType=Capsule, length=6.247, radius=1.1049, shapeName=hair_strand_back_03_l_capsule, center=<3.091 0 0>, rotation=<rot -90 0 0.0005>}]
 hair_strand_back_03_r [{
   shapeType=Capsule
   length=5.7371
   radius=1.1049
   shapeName=hair_strand_back_01_l_capsule
   center=<3.2362 0 0.1552>
   rotation=<rot -90 179.9996 -179.9995>
  }]
 hair_strand_back_04_l [{
   shapeType=Capsule
   length=6.247
   radius=1.1049
   shapeName=hair_strand_back_04_l_capsule
   center=<3.091 0 0>
   rotation=<rot -90 179.6835 -179.6832>
  }]
 hair_strand_back_04_r [{
   shapeType=Capsule
   length=5.7371
   radius=1.1049
   shapeName=hair_strand_back_01_l_capsule
   center=<3.2362 0 0.1552>
   rotation=<rot -90 179.9996 -179.9995>
  }]
 hair_strand_back_05_l [{
   shapeType=Capsule
   length=6.247
   radius=1.1049
   shapeName=hair_strand_back_05_l_capsule
   center=<3.091 0 0>
   rotation=<rot -90 179.9999 -179.9995>
  }]
 hair_strand_back_05_r [{
   shapeType=Capsule
   length=5.7371
   radius=1.1049
   shapeName=hair_strand_back_01_l_capsule
   center=<3.2362 0 0.1552>
   rotation=<rot -90 179.9996 -179.9995>
  }]
 hair_strand_back_06_l [{
   shapeType=Capsule
   length=6.247
   radius=1.1049
   shapeName=hair_strand_back_06_l_capsule
   center=<3.091 0 0>
   rotation=<rot -90 180 -179.9997>
  }]
 hair_strand_back_06_r [{
   shapeType=Capsule
   length=5.7371
   radius=1.1049
   shapeName=hair_strand_back_01_l_capsule
   center=<3.2362 0 0.1552>
   rotation=<rot -90 179.9996 -179.9995>
  }]
 hair_strand_back_mstr_l [{shapeType=Capsule, length=0.3857, radius=0.5, shapeName=hair_strand_back_mstr_l_capsule}]
 hair_strand_back_mstr_r [{shapeType=Capsule, length=0.3857, radius=0.5, shapeName=hair_strand_back_mstr_l_capsule, rotation=<rot 0 0 -180>}]
 head [{shapeType=Capsule, length=28.536, radius=9.5128, shapeName=head_capsule}]
 spine_05 [{shapeType=Capsule, length=17.6041, radius=7.1037, shapeName=spine_05_capsule, center=<4.5403 -0.6093 0>}]
}
constraints [
 {bone1Name=hair_strand_back_06_l, bone2Name=hair_strand_back_05_l, swing1LimitDegrees=25, swing2LimitDegrees=25, twistLimitDegrees=2}
 {bone1Name=hair_strand_back_05_l, bone2Name=hair_strand_back_04_l, swing1LimitDegrees=25, swing2LimitDegrees=25, twistLimitDegrees=2}
 {bone1Name=hair_strand_back_04_l, bone2Name=hair_strand_back_03_l, swing1LimitDegrees=25, swing2LimitDegrees=25, twistLimitDegrees=2}
 {bone1Name=hair_strand_back_03_l, bone2Name=hair_strand_back_02_l, swing1LimitDegrees=25, swing2LimitDegrees=25, twistLimitDegrees=2}
 {bone1Name=hair_strand_back_02_l, bone2Name=hair_strand_back_01_l, swing1LimitDegrees=25, swing2LimitDegrees=25, twistLimitDegrees=2}
 {bone1Name=hair_strand_back_01_l, bone2Name=hair_strand_back_mstr_l, swing1LimitDegrees=10, swing2LimitDegrees=5, twistLimitDegrees=2}
 {bone1Name=hair_strand_back_01_r, bone2Name=hair_strand_back_mstr_r, swing1LimitDegrees=10, swing2LimitDegrees=5, twistLimitDegrees=2}
 {bone1Name=hair_strand_back_02_r, bone2Name=hair_strand_back_01_r, swing1LimitDegrees=25, swing2LimitDegrees=25, twistLimitDegrees=2}
 {bone1Name=hair_strand_back_03_r, bone2Name=hair_strand_back_02_r, swing1LimitDegrees=25, swing2LimitDegrees=25, twistLimitDegrees=2}
 {bone1Name=hair_strand_back_04_r, bone2Name=hair_strand_back_03_r, swing1LimitDegrees=25, swing2LimitDegrees=25, twistLimitDegrees=2}
 {bone1Name=hair_strand_back_05_r, bone2Name=hair_strand_back_04_r, swing1LimitDegrees=25, swing2LimitDegrees=25, twistLimitDegrees=2}
 {bone1Name=hair_strand_back_06_r, bone2Name=hair_strand_back_05_r, swing1LimitDegrees=25, swing2LimitDegrees=25, twistLimitDegrees=2}
]
