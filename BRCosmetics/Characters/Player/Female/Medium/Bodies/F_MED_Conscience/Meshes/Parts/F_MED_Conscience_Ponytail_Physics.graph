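# Physics asset for the ponytail part of the F_MED_Conscience female character, defining its collision and jointed motion.
# Capsule bodies cover a four-bone tip chain (dyn_ponytail_tip_02 to 05), alongside upper spine, right clavicle and right upper arm capsules.
# Three constraints link neighboring ponytail bodies with tight twist and limited swing, widest at the last joint.
# The right-side clavicle and arm shapes carry left-side names.
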PhysicsAsset /BRCosmetics/Characters/Player/Female/Medium/Bodies/F_MED_Conscience/Meshes/Parts/F_MED_Conscience_Ponytail_Physics
bodies {
 clavicle_r [{
   shapeType=Capsule
   length=9.1739
   radius=4.9934
   shapeName=clavicle_l_capsule
   center=<-4.961 1.2493 -0.1575>
   rotation=<rot -75 179.9984 0.0012>
  }]
 dyn_ponytail_tip_02 [{shapeType=Capsule, length=0.2546, radius=1.8792, shapeName=dyn_ponytail_tip_02_capsule}]
 dyn_ponytail_tip_03 [{
   shapeType=Capsule
   length=3.4256
   radius=1.3312
   shapeName=dyn_ponytail_tip_03_capsule
   center=<2.4931 0 0>
   rotation=<rot -90 -0.0007 0.0003>
  }]
 dyn_ponytail_tip_04 [{shapeType=Capsule, length=3.4218, radius=1.335, shapeName=dyn_ponytail_tip_04_capsule, center=<2.4931 0 0>, rotation=<rot -90 0 -0.0003>}]
 dyn_ponytail_tip_05 [{
   shapeType=Capsule
   length=12.6708
   radius=1.6153
   shapeName=dyn_ponytail_tip_05_capsule
   center=<6.4836 0 -0.5325>
   rotation=<rot -85 179.9999 179.9998>
  }]
 spine_05 [
  {
   shapeType=Capsule
   length=11.4376
   radius=6.5064
   shapeName=spine_05_capsule
   center=<-4.2803 4.0042 -0.6263>
   rotation=<rot -6.7472 146.0883 174.2438>
  }
  {
   shapeType=Capsule
   length=11.4376
   radius=6.5064
   shapeName=spine_05_capsule
   center=<4.1204 -1.6433 0.5713>
   rotation=<rot -6.7472 146.0883 174.2438>
  }
 ]
 upperarm_r [{
   shapeType=Capsule
   length=24.052
   radius=4.5834
   shapeName=upperarm_l_capsule
   center=<-9.8053 -0.0001 0.3859>
   rotation=<rot -85 0.0039 179.9964>
  }]
}
constraints [
 {bone1Name=dyn_ponytail_tip_05, bone2Name=dyn_ponytail_tip_04, swing1LimitDegrees=30, swing2LimitDegrees=30, twistLimitDegrees=2}
 {bone1Name=dyn_ponytail_tip_04, bone2Name=dyn_ponytail_tip_03, swing1LimitDegrees=20, swing2LimitDegrees=20, twistLimitDegrees=2}
 {bone1Name=dyn_ponytail_tip_03, bone2Name=dyn_ponytail_tip_02, swing1LimitDegrees=20, swing2LimitDegrees=20, twistLimitDegrees=2}
]
