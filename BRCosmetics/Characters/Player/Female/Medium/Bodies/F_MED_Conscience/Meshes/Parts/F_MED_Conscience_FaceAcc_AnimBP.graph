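# Post-copy animation Blueprint for the F_MED_Conscience face accessory and hair.
# It layers the Head part's pose from the collar bones onto the parent's, then runs a FaceAcc Control Rig.
# A backbling check on customization load switches between a back-hair-strands Control Rig and FaceAcc rigid-body simulation.
# Ponytail and extra rigid-body passes always follow, with all three simulations using base-bone space and a gravity override updated each frame.
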AnimBlueprint /BRCosmetics/Characters/Player/Female/Medium/Bodies/F_MED_Conscience/Meshes/Parts/F_MED_Conscience_FaceAcc_AnimBP
target_skeleton=&/F_MED_Conscience_FaceAcc_Skeleton
graphs [
 AnimGraph {nodes [
   n1 = Get body_mesh() -> (body_mesh:Skeletal Mesh Component Object Reference)
   n2 = Control Rig!(
    Source=Layered blend per bone(BasePose=Copy Pose From Mesh() {bUseAttachedParent=true}, BlendPoses_0=Copy Pose From Mesh(SourceMeshComponent=@n1)) {layerSetup [{branchFilters [{boneName=dfrm_collar_l}, {boneName=dfrm_collar_r}]}]}
    Alpha
    is_front_end:bool=Get bIsInFrontEnd()
    is_playing_emote:bool=Get bIsPlayingEmote()
   ) -> (Pose) {
    alphaBoolBlend {blendOption=Linear}
    controlRigAssetReference {blueprintRigClass=&/F_MED_Conscience_FaceAcc_CtrlRig.F_MED_Conscience_FaceAcc_CtrlRig_C}
   }
   n3 = Control Rig!(Source=@n2, Alpha=Get is_wearing_backbling() -> (is_wearing_backbling:double)) -> (Pose) {
    alphaBoolBlend {blendOption=Linear}
    controlRigAssetReference {blueprintRigClass=&/F_MED_Conscience_Back_Hair_Strands_CtrlRig.F_MED_Conscience_Back_Hair_Strands_CtrlRig_C}
   }
   n4 = Get Gravity Override()
   n5 = RigidBody!(
    ComponentPose=Local To Component(LocalPose=@n3)
    Alpha=Get is_not_wearing_backbling() -> (is_not_wearing_backbling:double)
    OverrideWorldGravity:Vector=@n4
    ExternalForce
   ) -> (Pose) {
    bOverrideWorldGravity=true
    overridePhysicsAsset=&/F_MED_Conscience_FaceAcc_Physics
    simulationSpace=BaseBoneSpace
    alphaBoolBlend {blendOption=Linear}
    baseBoneRef {boneName=head}
    componentAppliedLinearAccClamp=<3000 3000 3000>
    componentLinearAccScale=<0.4 0.3 0.3>
    componentLinearVelScale=<0.4 0.3 0.3>
    simSpaceSettings {maxAngularAcceleration=1000, maxAngularVelocity=3, maxLinearAcceleration=250, maxLinearVelocity=500, velocityScaleZ=0, worldAlpha=1}
   }
   n6 = RigidBody(ComponentPose=@n5, OverrideWorldGravity=@n4) {
    bOverrideWorldGravity=true
    overridePhysicsAsset=&/F_MED_Conscience_Ponytail_Physics
    simulationSpace=BaseBoneSpace
    alphaBoolBlend {blendOption=Linear}
    baseBoneRef {boneName=spine_05}
    componentAppliedLinearAccClamp=<3000 3000 3000>
    componentLinearAccScale=<0.4 0.3 0.3>
    componentLinearVelScale=<0.4 0.3 0.3>
    simSpaceSettings {maxAngularAcceleration=1000, maxAngularVelocity=3, maxLinearAcceleration=250, maxLinearVelocity=500, velocityScaleZ=0, worldAlpha=1}
   }
   n7 = RigidBody(ComponentPose=@n6, OverrideWorldGravity=@n4) {
    bOverrideWorldGravity=true
    overridePhysicsAsset=&/F_MED_Conscience_FaceAcc_Extra_Physics
    simulationSpace=BaseBoneSpace
    alphaBoolBlend {blendOption=Linear}
    baseBoneRef {boneName=head}
    componentAppliedLinearAccClamp=<3000 3000 3000>
    componentLinearAccScale=<0.4 0.3 0.3>
    componentLinearVelScale=<0.4 0.3 0.3>
    simSpaceSettings {maxAngularAcceleration=1000, maxAngularVelocity=3, maxLinearAcceleration=250, maxLinearVelocity=500, velocityScaleZ=0, worldAlpha=1}
   }
   Output Pose(Result=Component To Local(ComponentPose=@n7)) {layerGroup=None}
  ]}
 EventGraph {nodes [
   Event BlueprintUpdateAnimation()
   > n1 = Sequence()
   > n2 = Cast To PlayerPawn_Athena(Object=TryGetPawnOwner())
   > Set body_mesh(body_mesh:Skeletal Mesh Component Object Reference=GetSkeletalMeshForPartType(self=@n2, PartType=Head)) -> (Output_Get:Skeletal Mesh Component Object Reference)
   n3 = Gravity Override State Function DEPRECATED(execute=@n1.then_1, CharacterPartAnimInstance=Self-Reference(), GameTimeInSeconds=GetTimeSeconds!() -> (ReturnValue:double))
   > Set Gravity Override(Gravity Override=@n3)
   Event OnCharacterCustomizationLoaded()
   > n4 = Branch(Condition=CheckForBackbling())
   > Set is_not_wearing_backbling(is_not_wearing_backbling:double=0) -> (Output_Get:double)
   > Set is_wearing_backbling(is_wearing_backbling:double=1) -> (Output_Get:double)
   Set is_not_wearing_backbling(execute=@n4.else, is_not_wearing_backbling=1)
   > Set is_wearing_backbling(is_wearing_backbling=0)
  ]}
]
variables [
 body_mesh {type=SkeletalMeshComponent}
 Gravity Override {type=Vector}
 is_not_wearing_backbling {type=number}
 is_wearing_backbling {type=number}
]
